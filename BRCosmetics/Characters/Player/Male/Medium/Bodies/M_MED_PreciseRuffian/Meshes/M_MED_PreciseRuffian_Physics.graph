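# Physics asset for the male medium PreciseRuffian body cosmetic, setting up jointed coat and front-cloth motion.
# Capsules cover pelvis, lower spine, legs and feet, plus a large zero-length one offset from root.
# Back, side and front coat chains and split front cloth hang from base anchor bones on spine_02 or pelvis, with mostly limited or locked joints.
# Angularly free cross-constraints tie neighbouring coat hem tips and both front-cloth tips together.
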PhysicsAsset /BRCosmetics/Characters/Player/Male/Medium/Bodies/M_MED_PreciseRuffian/Meshes/M_MED_PreciseRuffian_Physics
bodies {
 base_back_coat_00_c [{shapeType=Capsule, length=1, radius=1}]
 base_coat_support_l [{shapeType=Capsule, length=0.505, radius=0.505, shapeName=base_coat_support_l_capsule, rotation=<rot 0 -90 0>}]
 base_coat_support_r [{shapeType=Capsule, length=0.505, radius=0.505, shapeName=base_coat_support_r_capsule, rotation=<rot 0 -90 0>}]
 base_frontCloth_c [{shapeType=Capsule, length=1, radius=1, shapeName=base_frontCloth_c_capsule, rotation=<rot 90 0 0>}]
 base_front_coat_01_l [{shapeType=Capsule, length=0.505, radius=0.505, shapeName=base_front_coat_01_l_capsule, rotation=<rot 0 -90 0>}]
 base_front_coat_01_r [{shapeType=Capsule, length=0.505, radius=0.505, shapeName=base_front_coat_01_r_capsule, rotation=<rot 0 -90 0>}]
 base_side_coat_00_l [{shapeType=Capsule, length=1, radius=1}]
 base_side_coat_00_r [{shapeType=Capsule, length=1, radius=1}]
 calf_l [{shapeType=Capsule, length=35.5, radius=12.5, center=<-20 -2 0>, rotation=<rot 90 0 0>}]
 calf_r [{shapeType=Capsule, length=35.5, radius=12.5, center=<20 2 0>, rotation=<rot 90 0 0>}]
 dyn_back_coat_00_c [{shapeType=Capsule, length=5, radius=8, center=<3 0 5>, rotation=<rot 0 0 90>}]
 dyn_back_coat_01_c [{shapeType=Capsule, length=5, radius=9.5, center=<3 0 6.5>, rotation=<rot 0 0 90>}]
 dyn_back_coat_02_c [{shapeType=Capsule, length=5, radius=10.5, center=<3 0 6.5>, rotation=<rot 0 0 90>}]
 dyn_back_coat_03_c [{shapeType=Capsule, length=4, radius=4, center=<4 0 0>, rotation=<rot 90 0 0>}]
 dyn_back_coat_a_00_l [{shapeType=Capsule, length=20, radius=6.5, center=<15 5 1>, rotation=<rot 90 0 0>}]
 dyn_back_coat_a_00_r [{shapeType=Capsule, length=20, radius=6.5, center=<15 -5 1>, rotation=<rot 90 0 0>}]
 dyn_frontCloth_00_l [{shapeType=Capsule, length=4, radius=4, shapeName=dyn_frontCloth_00_l_capsule, center=<6 -4 0>, rotation=<rot 90 0 0>}]
 dyn_frontCloth_00_r [{shapeType=Capsule, length=4, radius=4, shapeName=dyn_frontCloth_00_r_capsule, center=<6 -2 0>, rotation=<rot 90 0 0>}]
 dyn_frontCloth_01_l [{shapeType=Capsule, length=4, radius=4, shapeName=dyn_frontCloth_01_l_capsule, center=<6 -4 0>, rotation=<rot 90 0 0>}]
 dyn_frontCloth_01_r [{shapeType=Capsule, length=4, radius=4, shapeName=dyn_frontCloth_01_r_capsule, center=<6 -4 0>, rotation=<rot 90 0 0>}]
 dyn_frontCloth_02_l [
  {shapeType=Capsule, length=4, radius=4.5, shapeName=dyn_frontCloth_02_l_capsule, center=<6 -4 0>, rotation=<rot 90 0 0>}
  {shapeType=Capsule, length=8, radius=5, center=<5 -4 5>}
 ]
 dyn_frontCloth_02_r [
  {shapeType=Capsule, length=4, radius=4.5, shapeName=dyn_frontCloth_02_r_capsule, center=<6 -1.5 0>, rotation=<rot 90 0 0>}
  {shapeType=Capsule, length=8, radius=5, center=<5 -1.5 -8>}
 ]
 dyn_frontCloth_03_l [
  {shapeType=Capsule, length=4, radius=5, shapeName=dyn_frontCloth_03_l_capsule, center=<6 -4 0>, rotation=<rot 90 0 0>}
  {shapeType=Capsule, length=8, radius=5, center=<5 -4 8>}
 ]
 dyn_frontCloth_03_r [
  {shapeType=Capsule, length=5.5, radius=5, shapeName=dyn_frontCloth_03_r_capsule, center=<7.5838 -1.5 0>, rotation=<rot 90 0 0>}
  {shapeType=Capsule, length=8, radius=5, center=<5 -1.5 -8>}
 ]
 dyn_frontCloth_c [{shapeType=Capsule, length=1, radius=1, shapeName=dyn_frontCloth_c_capsule, rotation=<rot 90 0 0>}]
 dyn_front_coat_01_l [{shapeType=Capsule, length=21, radius=6.5, center=<10 -6 0>, rotation=<rot 0 0 55>}]
 dyn_front_coat_01_r [{shapeType=Capsule, length=21, radius=6.5, center=<10 3.854 1>, rotation=<rot 0 0 125>}]
 dyn_front_coat_02_l [{shapeType=Capsule, length=20.5, radius=7, center=<15 -7 1>, rotation=<rot 0 0 55>}]
 dyn_front_coat_02_r [{shapeType=Capsule, length=20.5, radius=7, center=<15 7 1>, rotation=<rot 0 0 125>}]
 dyn_front_coat_03_l [{shapeType=Capsule, length=19.5, radius=8, center=<10 -6 2>, rotation=<rot 0 0 55>}]
 dyn_front_coat_03_r [{shapeType=Capsule, length=19.5, radius=8, center=<10 6 2>, rotation=<rot 0 0 125>}]
 dyn_front_coat_04_l [{shapeType=Capsule, length=4, radius=4, center=<4 0 0>, rotation=<rot 90 0 0>}]
 dyn_front_coat_04_r [{shapeType=Capsule, length=4, radius=4, center=<4 0 0>, rotation=<rot 90 0 0>}]
 dyn_side_coat_00_l [{shapeType=Capsule, length=4, radius=4, center=<4 0 0>, rotation=<rot 90 0 0>}]
 dyn_side_coat_00_r [{shapeType=Capsule, length=4, radius=4, center=<4 0 0>, rotation=<rot 90 0 0>}]
 dyn_side_coat_01_l [{shapeType=Capsule, length=23, radius=9.5, center=<8 0 12.5>, rotation=<rot 0 0 75>}]
 dyn_side_coat_01_r [{shapeType=Capsule, length=23, radius=9.5, center=<8 0 12>, rotation=<rot 0 0 95>}]
 dyn_side_coat_02_l [{shapeType=Capsule, length=26, radius=9.5, center=<6.5 0 9>, rotation=<rot 0 0 75>}]
 dyn_side_coat_02_r [{shapeType=Capsule, length=26, radius=9.5, center=<6.5 -0.1889 10.5>, rotation=<rot 0 0 95>}]
 dyn_side_coat_03_l [{shapeType=Capsule, length=27.5, radius=9.5, center=<10 0 8>, rotation=<rot 0 0 75>}]
 dyn_side_coat_03_r [{shapeType=Capsule, length=27.5, radius=9.5, center=<15 0.3032 8>, rotation=<rot 0 0 95>}]
 dyn_side_coat_04_l [{shapeType=Capsule, length=4, radius=4, center=<4 0 0>, rotation=<rot 90 0 0>}]
 dyn_side_coat_04_r [{shapeType=Capsule, length=4, radius=4, center=<4 0 0>, rotation=<rot 90 0 0>}]
 foot_l [{shapeType=Capsule, length=9.5, radius=15.5, center=<-4 6 0>, rotation=<rot 90 0 80>}]
 foot_r [{shapeType=Capsule, length=9.5, radius=15.5, center=<4 -6 1.9991>, rotation=<rot 90 -165.9638 -85.9638>}]
 pelvis [{shapeType=Capsule, length=4.5, radius=13.5, center=<0 1.5297 0>}]
 root [{shapeType=Capsule, length=0, radius=125, center=<-15 0 -113.1506>}]
 spine_01 [{shapeType=Capsule, length=4.5, radius=11.5, center=<4 1.9692 0>}]
 spine_02 [{shapeType=Capsule, length=4.5, radius=11.5, center=<4 1.9692 0>}]
 thigh_l [{shapeType=Capsule, length=29.5, radius=10.5, center=<-20 -2 -1>, rotation=<rot 90 0 0>}]
 thigh_r [{shapeType=Capsule, length=29.5, radius=10.5, center=<20 2 1>, rotation=<rot 90 0 0>}]
}
constraints [
 {
  bone1Name=pelvis
  bone2Name=root
  swing1LimitDegrees=45
  swing1Motion=Locked
  swing2LimitDegrees=45
  swing2Motion=Locked
  twistLimitDegrees=45
  twistMotion=Locked
 }
 {
  bone1Name=base_back_coat_00_c
  bone2Name=spine_02
  swing1LimitDegrees=45
  swing1Motion=Locked
  swing2LimitDegrees=45
  swing2Motion=Locked
  twistLimitDegrees=45
  twistMotion=Locked
 }
 {bone1Name=dyn_back_coat_00_c, bone2Name=base_back_coat_00_c, swing1LimitDegrees=10, swing2LimitDegrees=20, twistLimitDegrees=5}
 {bone1Name=dyn_back_coat_01_c, bone2Name=dyn_back_coat_00_c, swing1LimitDegrees=15, swing2LimitDegrees=25, twistLimitDegrees=15}
 {bone1Name=dyn_back_coat_02_c, bone2Name=dyn_back_coat_01_c, swing1LimitDegrees=5, swing2LimitDegrees=30, twistLimitDegrees=15}
 {bone1Name=dyn_back_coat_03_c, bone2Name=dyn_back_coat_02_c, swing1LimitDegrees=5, swing2LimitDegrees=5, twistLimitDegrees=5}
 {
  bone1Name=dyn_side_coat_04_l
  bone2Name=dyn_back_coat_03_c
  swing1LimitDegrees=45
  swing1Motion=Free
  swing2LimitDegrees=45
  swing2Motion=Free
  twistLimitDegrees=45
  twistMotion=Free
 }
 {
  bone1Name=dyn_side_coat_04_r
  bone2Name=dyn_back_coat_03_c
  swing1LimitDegrees=45
  swing1Motion=Free
  swing2LimitDegrees=45
  swing2Motion=Free
  twistLimitDegrees=45
  twistMotion=Free
 }
 {
  bone1Name=base_coat_support_l
  bone2Name=pelvis
  swing1LimitDegrees=45
  swing1Motion=Free
  swing2LimitDegrees=45
  swing2Motion=Free
  twistLimitDegrees=45
  twistMotion=Free
 }
 {
  bone1Name=base_front_coat_01_l
  bone2Name=base_coat_support_l
  swing1LimitDegrees=45
  swing1Motion=Free
  swing2LimitDegrees=45
  swing2Motion=Free
  twistLimitDegrees=45
  twistMotion=Free
 }
 {
  bone1Name=dyn_front_coat_01_l
  bone2Name=base_front_coat_01_l
  swing1LimitDegrees=5
  swing2LimitDegrees=40
  twistLimitDegrees=45
  twistMotion=Locked
 }
 {
  bone1Name=dyn_front_coat_02_l
  bone2Name=dyn_front_coat_01_l
  swing1LimitDegrees=5
  swing2LimitDegrees=50
  twistLimitDegrees=45
  twistMotion=Locked
 }
 {
  bone1Name=dyn_front_coat_03_l
  bone2Name=dyn_front_coat_02_l
  swing1LimitDegrees=20
  swing2LimitDegrees=40
  twistLimitDegrees=45
  twistMotion=Locked
 }
 {bone1Name=dyn_front_coat_04_l, bone2Name=dyn_front_coat_03_l, swing1LimitDegrees=45, swing2LimitDegrees=45, twistLimitDegrees=5}
 {
  bone1Name=dyn_front_coat_04_l
  bone2Name=dyn_side_coat_04_l
  swing1LimitDegrees=45
  swing1Motion=Free
  swing2LimitDegrees=45
  swing2Motion=Free
  twistLimitDegrees=45
  twistMotion=Free
 }
 {
  bone1Name=base_coat_support_r
  bone2Name=pelvis
  swing1LimitDegrees=45
  swing1Motion=Free
  swing2LimitDegrees=45
  swing2Motion=Free
  twistLimitDegrees=45
  twistMotion=Free
 }
 {
  bone1Name=base_front_coat_01_r
  bone2Name=base_coat_support_r
  swing1LimitDegrees=45
  swing1Motion=Free
  swing2LimitDegrees=45
  swing2Motion=Free
  twistLimitDegrees=45
  twistMotion=Free
 }
 {
  bone1Name=dyn_front_coat_01_r
  bone2Name=base_front_coat_01_r
  swing1LimitDegrees=5
  swing2LimitDegrees=40
  twistLimitDegrees=45
  twistMotion=Locked
 }
 {
  bone1Name=dyn_front_coat_02_r
  bone2Name=dyn_front_coat_01_r
  swing1LimitDegrees=5
  swing2LimitDegrees=50
  twistLimitDegrees=45
  twistMotion=Locked
 }
 {
  bone1Name=dyn_front_coat_03_r
  bone2Name=dyn_front_coat_02_r
  swing1LimitDegrees=20
  swing2LimitDegrees=40
  twistLimitDegrees=45
  twistMotion=Locked
 }
 {bone1Name=dyn_front_coat_04_r, bone2Name=dyn_front_coat_03_r, swing1LimitDegrees=45, swing2LimitDegrees=45, twistLimitDegrees=5}
 {
  bone1Name=dyn_front_coat_04_r
  bone2Name=dyn_side_coat_04_r
  swing1LimitDegrees=45
  swing1Motion=Free
  swing2LimitDegrees=45
  swing2Motion=Free
  twistLimitDegrees=45
  twistMotion=Free
 }
 {
  bone1Name=base_side_coat_00_l
  bone2Name=spine_02
  swing1LimitDegrees=45
  swing1Motion=Locked
  swing2LimitDegrees=45
  swing2Motion=Locked
  twistLimitDegrees=45
  twistMotion=Locked
 }
 {
  bone1Name=dyn_side_coat_00_l
  bone2Name=base_side_coat_00_l
  swing1LimitDegrees=45
  swing1Motion=Locked
  swing2LimitDegrees=45
  twistLimitDegrees=15
 }
 {
  bone1Name=dyn_side_coat_01_l
  bone2Name=dyn_side_coat_00_l
  swing1LimitDegrees=10
  swing2LimitDegrees=25
  twistLimitDegrees=5
  twistMotion=Locked
 }
 {
  bone1Name=dyn_side_coat_02_l
  bone2Name=dyn_side_coat_01_l
  swing1LimitDegrees=1
  swing2LimitDegrees=40
  twistLimitDegrees=5
  twistMotion=Locked
 }
 {
  bone1Name=dyn_side_coat_03_l
  bone2Name=dyn_side_coat_02_l
  swing1LimitDegrees=1
  swing2LimitDegrees=40
  twistLimitDegrees=5
  twistMotion=Locked
 }
 {bone1Name=dyn_side_coat_04_l, bone2Name=dyn_side_coat_03_l, swing1LimitDegrees=45, swing2LimitDegrees=45, twistLimitDegrees=5}
 {
  bone1Name=base_side_coat_00_r
  bone2Name=spine_02
  swing1LimitDegrees=45
  swing1Motion=Locked
  swing2LimitDegrees=45
  swing2Motion=Locked
  twistLimitDegrees=45
  twistMotion=Locked
 }
 {bone1Name=dyn_side_coat_00_r, bone2Name=base_side_coat_00_r, swing1LimitDegrees=45, swing2LimitDegrees=45, twistLimitDegrees=15}
 {
  bone1Name=dyn_side_coat_01_r
  bone2Name=dyn_side_coat_00_r
  swing1LimitDegrees=10
  swing2LimitDegrees=25
  twistLimitDegrees=15
  twistMotion=Locked
 }
 {
  bone1Name=dyn_side_coat_02_r
  bone2Name=dyn_side_coat_01_r
  swing1LimitDegrees=1
  swing2LimitDegrees=25
  twistLimitDegrees=15
  twistMotion=Locked
 }
 {
  bone1Name=dyn_side_coat_03_r
  bone2Name=dyn_side_coat_02_r
  swing1LimitDegrees=1
  swing2LimitDegrees=25
  twistLimitDegrees=15
  twistMotion=Locked
 }
 {bone1Name=dyn_side_coat_04_r, bone2Name=dyn_side_coat_03_r, swing1LimitDegrees=45, swing2LimitDegrees=45, twistLimitDegrees=5}
 {
  bone1Name=spine_01
  bone2Name=pelvis
  swing1LimitDegrees=45
  swing1Motion=Locked
  swing2LimitDegrees=45
  swing2Motion=Locked
  twistLimitDegrees=45
  twistMotion=Locked
 }
 {
  bone1Name=spine_02
  bone2Name=spine_01
  swing1LimitDegrees=45
  swing1Motion=Locked
  swing2LimitDegrees=45
  swing2Motion=Locked
  twistLimitDegrees=45
  twistMotion=Locked
 }
 {
  bone1Name=thigh_l
  bone2Name=pelvis
  swing1LimitDegrees=45
  swing1Motion=Locked
  swing2LimitDegrees=45
  swing2Motion=Locked
  twistLimitDegrees=45
  twistMotion=Locked
 }
 {
  bone1Name=calf_l
  bone2Name=thigh_l
  swing1LimitDegrees=45
  swing1Motion=Locked
  swing2LimitDegrees=45
  swing2Motion=Locked
  twistLimitDegrees=45
  twistMotion=Locked
 }
 {
  bone1Name=foot_l
  bone2Name=calf_l
  swing1LimitDegrees=45
  swing1Motion=Free
  swing2LimitDegrees=45
  swing2Motion=Free
  twistLimitDegrees=45
  twistMotion=Free
 }
 {
  bone1Name=thigh_r
  bone2Name=pelvis
  swing1LimitDegrees=45
  swing1Motion=Locked
  swing2LimitDegrees=45
  swing2Motion=Locked
  twistLimitDegrees=45
  twistMotion=Locked
 }
 {
  bone1Name=calf_r
  bone2Name=thigh_r
  swing1LimitDegrees=45
  swing1Motion=Locked
  swing2LimitDegrees=45
  swing2Motion=Locked
  twistLimitDegrees=45
  twistMotion=Locked
 }
 {
  bone1Name=foot_r
  bone2Name=calf_r
  swing1LimitDegrees=45
  swing1Motion=Free
  swing2LimitDegrees=45
  swing2Motion=Free
  twistLimitDegrees=45
  twistMotion=Free
 }
 {bone1Name=dyn_back_coat_a_00_r, bone2Name=dyn_back_coat_02_c, swing1LimitDegrees=25, swing2LimitDegrees=25, twistLimitDegrees=1}
 {bone1Name=dyn_back_coat_a_00_l, bone2Name=dyn_back_coat_02_c, swing1LimitDegrees=25, swing2LimitDegrees=25, twistLimitDegrees=1}
 {bone1Name=dyn_frontCloth_03_r, bone2Name=dyn_frontCloth_02_r, swing1LimitDegrees=30, swing2LimitDegrees=15, twistLimitDegrees=5}
 {
  bone1Name=dyn_frontCloth_03_r
  bone2Name=dyn_frontCloth_03_l
  swing1LimitDegrees=45
  swing1Motion=Free
  swing2LimitDegrees=45
  swing2Motion=Free
  twistLimitDegrees=45
  twistMotion=Free
 }
 {
  bone1Name=dyn_frontCloth_03_l
  bone2Name=dyn_frontCloth_03_r
  swing1LimitDegrees=45
  swing1Motion=Free
  swing2LimitDegrees=45
  swing2Motion=Free
  twistLimitDegrees=45
  twistMotion=Free
 }
 {
  bone1Name=base_frontCloth_c
  bone2Name=pelvis
  swing1LimitDegrees=45
  swing1Motion=Locked
  swing2LimitDegrees=45
  swing2Motion=Locked
  twistLimitDegrees=45
  twistMotion=Locked
 }
 {bone1Name=dyn_frontCloth_c, bone2Name=base_frontCloth_c, swing1LimitDegrees=45, swing2LimitDegrees=45, twistLimitDegrees=45}
 {bone1Name=dyn_frontCloth_00_l, bone2Name=dyn_frontCloth_c, swing1LimitDegrees=45, swing2LimitDegrees=45, twistLimitDegrees=45}
 {bone1Name=dyn_frontCloth_00_r, bone2Name=dyn_frontCloth_c, swing1LimitDegrees=45, swing2LimitDegrees=45, twistLimitDegrees=45}
 {bone1Name=dyn_frontCloth_01_l, bone2Name=dyn_frontCloth_00_l, swing1LimitDegrees=45, swing2LimitDegrees=15, twistLimitDegrees=5}
 {bone1Name=dyn_frontCloth_02_l, bone2Name=dyn_frontCloth_01_l, swing1LimitDegrees=45, swing2LimitDegrees=15, twistLimitDegrees=5}
 {bone1Name=dyn_frontCloth_03_l, bone2Name=dyn_frontCloth_02_l, swing1LimitDegrees=30, swing2LimitDegrees=15, twistLimitDegrees=5}
 {bone1Name=dyn_frontCloth_01_r, bone2Name=dyn_frontCloth_00_r, swing1LimitDegrees=45, swing2LimitDegrees=15, twistLimitDegrees=5}
 {bone1Name=dyn_frontCloth_02_r, bone2Name=dyn_frontCloth_01_r, swing1LimitDegrees=45, swing2LimitDegrees=15, twistLimitDegrees=5}
]
