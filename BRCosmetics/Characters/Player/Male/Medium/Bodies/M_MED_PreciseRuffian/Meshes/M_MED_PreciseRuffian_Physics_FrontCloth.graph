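# Physics asset for the front cloth of the PreciseRuffian medium male body, simulating a hanging garment with rigid-body bone chains.
# A locked base on the pelvis carries a center bone branching into left and right four-segment capsule chains, with tighter angular limits down each chain.
# Pelvis and locked thigh capsules provide leg volumes, and lower segments add second offset capsules.
# Reciprocal free-rotation constraints tie the two chain tips together.
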PhysicsAsset /BRCosmetics/Characters/Player/Male/Medium/Bodies/M_MED_PreciseRuffian/Meshes/M_MED_PreciseRuffian_Physics_FrontCloth
bodies {
 base_frontCloth_c [{shapeType=Capsule, length=1, radius=1, shapeName=base_frontCloth_c_capsule, rotation=<rot 90 0 0>}]
 dyn_frontCloth_00_l [{shapeType=Capsule, length=4, radius=4, shapeName=dyn_frontCloth_00_l_capsule, center=<6 -4 0>, rotation=<rot 90 0 0>}]
 dyn_frontCloth_00_r [{shapeType=Capsule, length=4, radius=4, shapeName=dyn_frontCloth_00_r_capsule, center=<6 -2 0>, rotation=<rot 90 0 0>}]
 dyn_frontCloth_01_l [{shapeType=Capsule, length=4, radius=4, shapeName=dyn_frontCloth_01_l_capsule, center=<6 -4 0>, rotation=<rot 90 0 0>}]
 dyn_frontCloth_01_r [{shapeType=Capsule, length=4, radius=4, shapeName=dyn_frontCloth_01_r_capsule, center=<6 -4 0>, rotation=<rot 90 0 0>}]
 dyn_frontCloth_02_l [
  {shapeType=Capsule, length=4, radius=4.5, shapeName=dyn_frontCloth_02_l_capsule, center=<6 -4 0>, rotation=<rot 90 0 0>}
  {shapeType=Capsule, length=8, radius=5, center=<5 -4 5>}
 ]
 dyn_frontCloth_02_r [
  {shapeType=Capsule, length=4, radius=4.5, shapeName=dyn_frontCloth_02_r_capsule, center=<6 -1.5 0>, rotation=<rot 90 0 0>}
  {shapeType=Capsule, length=8, radius=5, center=<5 -1.5 -8>}
 ]
 dyn_frontCloth_03_l [
  {shapeType=Capsule, length=4, radius=5, shapeName=dyn_frontCloth_03_l_capsule, center=<6 -4 0>, rotation=<rot 90 0 0>}
  {shapeType=Capsule, length=8, radius=5, center=<5 -4 8>}
 ]
 dyn_frontCloth_03_r [
  {shapeType=Capsule, length=5.5, radius=5, shapeName=dyn_frontCloth_03_r_capsule, center=<7.5838 -1.5 0>, rotation=<rot 90 0 0>}
  {shapeType=Capsule, length=8, radius=5, center=<5 -1.5 -8>}
 ]
 dyn_frontCloth_c [{shapeType=Capsule, length=1, radius=1, shapeName=dyn_frontCloth_c_capsule, rotation=<rot 90 0 0>}]
 pelvis [{shapeType=Capsule, length=9, radius=9, shapeName=pelvis_capsule}]
 thigh_l [{shapeType=Capsule, length=38, radius=10, shapeName=thigh_l_capsule, center=<-22 0 2>, rotation=<rot 90 0 0>}]
 thigh_r [{shapeType=Capsule, length=38, radius=10, shapeName=thigh_r_capsule, center=<22 0 -2>, rotation=<rot 90 0 0>}]
}
constraints [
 {
  bone1Name=base_frontCloth_c
  bone2Name=pelvis
  swing1LimitDegrees=45
  swing1Motion=Locked
  swing2LimitDegrees=45
  swing2Motion=Locked
  twistLimitDegrees=45
  twistMotion=Locked
 }
 {bone1Name=dyn_frontCloth_c, bone2Name=base_frontCloth_c, swing1LimitDegrees=45, swing2LimitDegrees=45, twistLimitDegrees=45}
 {bone1Name=dyn_frontCloth_00_l, bone2Name=dyn_frontCloth_c, swing1LimitDegrees=45, swing2LimitDegrees=45, twistLimitDegrees=45}
 {bone1Name=dyn_frontCloth_01_l, bone2Name=dyn_frontCloth_00_l, swing1LimitDegrees=45, swing2LimitDegrees=15, twistLimitDegrees=5}
 {bone1Name=dyn_frontCloth_02_l, bone2Name=dyn_frontCloth_01_l, swing1LimitDegrees=45, swing2LimitDegrees=15, twistLimitDegrees=5}
 {bone1Name=dyn_frontCloth_03_l, bone2Name=dyn_frontCloth_02_l, swing1LimitDegrees=30, swing2LimitDegrees=15, twistLimitDegrees=5}
 {bone1Name=dyn_frontCloth_00_r, bone2Name=dyn_frontCloth_c, swing1LimitDegrees=45, swing2LimitDegrees=45, twistLimitDegrees=45}
 {bone1Name=dyn_frontCloth_01_r, bone2Name=dyn_frontCloth_00_r, swing1LimitDegrees=45, swing2LimitDegrees=15, twistLimitDegrees=5}
 {bone1Name=dyn_frontCloth_02_r, bone2Name=dyn_frontCloth_01_r, swing1LimitDegrees=45, swing2LimitDegrees=15, twistLimitDegrees=5}
 {bone1Name=dyn_frontCloth_03_r, bone2Name=dyn_frontCloth_02_r, swing1LimitDegrees=30, swing2LimitDegrees=15, twistLimitDegrees=5}
 {
  bone1Name=thigh_l
  bone2Name=pelvis
  swing1LimitDegrees=45
  swing1Motion=Locked
  swing2LimitDegrees=45
  swing2Motion=Locked
  twistLimitDegrees=45
  twistMotion=Locked
 }
 {
  bone1Name=thigh_r
  bone2Name=pelvis
  swing1LimitDegrees=45
  swing1Motion=Locked
  swing2LimitDegrees=45
  swing2Motion=Locked
  twistLimitDegrees=45
  twistMotion=Locked
 }
 {
  bone1Name=dyn_frontCloth_03_r
  bone2Name=dyn_frontCloth_03_l
  swing1LimitDegrees=45
  swing1Motion=Free
  swing2LimitDegrees=45
  swing2Motion=Free
  twistLimitDegrees=45
  twistMotion=Free
 }
 {
  bone1Name=dyn_frontCloth_03_l
  bone2Name=dyn_frontCloth_03_r
  swing1LimitDegrees=45
  swing1Motion=Free
  swing2LimitDegrees=45
  swing2Motion=Free
  twistLimitDegrees=45
  twistMotion=Free
 }
]
