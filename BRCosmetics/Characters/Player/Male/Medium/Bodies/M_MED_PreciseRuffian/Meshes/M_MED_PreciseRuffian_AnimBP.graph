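# Post-process style AnimBP for the PreciseRuffian male player body, adding secondary motion to its coat, sleeves and hip bags.
# Copied parent pose runs through a Control Rig, sleeve AnimDynamics, then RigidBody coat simulation.
# Cloth, rigid-body and DBNO flags select full-coat, front-cloth or trail-controller poses, then constraints and additive offsets adjust bags and coat.
# EventGraph derives gravity, cloth settings and blend alphas from movement, emote, weapon and backbling state.
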AnimBlueprint /BRCosmetics/Characters/Player/Male/Medium/Bodies/M_MED_PreciseRuffian/Meshes/M_MED_PreciseRuffian_AnimBP
target_skeleton=&/M_MED_PreciseRuffian_Skeleton
graphs [
 AnimGraph {nodes [
   n1 = Blend Poses by bool(BlendPose_0=Use cached pose 'dbnoCache'(), BlendPose_1=Use cached pose 'trailCache'(), bActiveValue=Get bIsDBNO()) {blendType=Linear}
   n2 = Blend Poses by bool(BlendPose_0=Use cached pose 'frontCloth_Cache'(), BlendPose_1=Use cached pose 'physicsCoat_Cache'(), bActiveValue=Get bClothEnabled()) {blendType=Linear}
   n3 = Get bRigidBodyEnabled()
   n4 = "Constraint - base_hipBag_b_l (PPPPPPPP)"!(
    ComponentPose=Local To Component(LocalPose=Blend Poses by bool(BlendPose_0=@n2, BlendPose_1=@n1, bActiveValue=@n3) {blendType=Linear})
    Alpha=1
    ConstraintWeights_0=0.5193
    ConstraintWeights_1:float=0.2339
    ConstraintWeights_2:float=0.1271
    ConstraintWeights_3:float=0.0994
    ConstraintWeights_4:float=0.0084
    ConstraintWeights_5:float=0.0051
    ConstraintWeights_6:float=0.0037
    ConstraintWeights_7:float=0.0032
   ) -> (Pose) {
    alphaBoolBlend {blendOption=Linear}
    boneToModify {boneName=base_hipBag_b_l}
    constraintSetup [
     {targetBone {boneName=dyn_side_coat_00_l}}
     {targetBone {boneName=dyn_side_coat_01_l}}
     {targetBone {boneName=pelvis}}
     {targetBone {boneName=dyn_back_coat_00_c}}
     {targetBone {boneName=thigh_twist_01_l}}
     {targetBone {boneName=deform_glute_l}}
     {targetBone {boneName=spine_02}}
     {targetBone {boneName=spine_01}}
    ]
    constraintWeights [1,1,1,1,1,1,1,1]
   }
   n5 = "Constraint - base_hipBag_a_l (PPPPPPPP)"!(
    ComponentPose=@n4
    Alpha=1
    ConstraintWeights_0=0.4602
    ConstraintWeights_1:float=0.2366
    ConstraintWeights_2:float=0.1377
    ConstraintWeights_3:float=0.0912
    ConstraintWeights_4:float=0.0443
    ConstraintWeights_5:float=0.02
    ConstraintWeights_6:float=0.0054
    ConstraintWeights_7:float=0.0045
   ) -> (Pose) {
    alphaBoolBlend {blendOption=Linear}
    boneToModify {boneName=base_hipBag_a_l}
    constraintSetup [
     {targetBone {boneName=dyn_front_coat_01_l}}
     {targetBone {boneName=dyn_side_coat_00_l}}
     {targetBone {boneName=pelvis}}
     {targetBone {boneName=dyn_side_coat_01_l}}
     {targetBone {boneName=dyn_front_coat_02_l}}
     {targetBone {boneName=thigh_twist_01_l}}
     {targetBone {boneName=spine_02}}
     {targetBone {boneName=spine_01}}
    ]
    constraintWeights [1,1,1,1,1,1,1,1]
   }
   n6 = "Constraint - base_hipBag_b_r (PPPPPPPP)"!(
    ComponentPose=@n5
    Alpha=1
    ConstraintWeights_0=0.5235
    ConstraintWeights_1:float=0.2298
    ConstraintWeights_2:float=0.1316
    ConstraintWeights_3:float=0.0953
    ConstraintWeights_4:float=0.0089
    ConstraintWeights_5:float=0.005
    ConstraintWeights_6:float=0.0031
    ConstraintWeights_7:float=0.0027
   ) -> (Pose) {
    alphaBoolBlend {blendOption=Linear}
    boneToModify {boneName=base_hipBag_b_r}
    constraintSetup [
     {targetBone {boneName=dyn_side_coat_00_r}}
     {targetBone {boneName=dyn_side_coat_01_r}}
     {targetBone {boneName=pelvis}}
     {targetBone {boneName=dyn_back_coat_00_c}}
     {targetBone {boneName=thigh_twist_01_r}}
     {targetBone {boneName=deform_glute_r}}
     {targetBone {boneName=dyn_front_coat_01_r}}
     {targetBone {boneName=spine_02}}
    ]
    constraintWeights [1,1,1,1,1,1,1,1]
   }
   n7 = "Constraint - base_hipBag_a_r (PPPPPPPP)"!(
    ComponentPose=@n6
    Alpha=1
    ConstraintWeights_0=0.4415
    ConstraintWeights_1:float=0.2275
    ConstraintWeights_2:float=0.1273
    ConstraintWeights_3:float=0.1045
    ConstraintWeights_4:float=0.064
    ConstraintWeights_5:float=0.0188
    ConstraintWeights_6:float=0.0136
    ConstraintWeights_7:float=0.0028
   ) -> (Pose) {
    alphaBoolBlend {blendOption=Linear}
    boneToModify {boneName=base_hipBag_a_r}
    constraintSetup [
     {targetBone {boneName=dyn_front_coat_01_r}}
     {targetBone {boneName=dyn_side_coat_00_r}}
     {targetBone {boneName=dyn_side_coat_01_r}}
     {targetBone {boneName=pelvis}}
     {targetBone {boneName=dyn_front_coat_02_r}}
     {targetBone {boneName=thigh_twist_01_r}}
     {targetBone {boneName=deform_groin_r}}
     {targetBone {boneName=spine_02}}
    ]
    constraintWeights [1,1,1,1,1,1,1,1]
   }
   n8 = "Transform (Modify) Bone - Bone: base_front_coat_01_r"!(ComponentPose=@n7, bAlphaBoolEnabled:bool=True, Translation="0,0,0", Rotation="25,0,0", Scale="1,1,1") -> (Pose) {
    alphaInputType=Bool
    rotationMode=BMM_Additive
    rotationSpace=BCS_BoneSpace
    alphaBoolBlend {blendInTime=0.5, blendOption=Linear, blendOutTime=0.5}
    boneToModify {boneName=base_front_coat_01_r}
    rotation {pitch=25}
   }
   n9 = "Transform (Modify) Bone - Bone: base_side_coat_00_r"!(ComponentPose=@n8, bAlphaBoolEnabled:bool=True, Translation="0,0,0", Rotation="-5,0,0", Scale="1,1,1") -> (Pose) {
    alphaInputType=Bool
    rotationMode=BMM_Additive
    rotationSpace=BCS_BoneSpace
    alphaBoolBlend {blendInTime=0.5, blendOption=Linear, blendOutTime=0.5}
    boneToModify {boneName=base_side_coat_00_r}
    rotation {pitch=-5}
   }
   n10 = "Transform (Modify) Bone - Bone: base_back_coat_00_c"!(ComponentPose=@n9, bAlphaBoolEnabled:bool=True, Translation="0,0,0", Rotation="10,0,0", Scale="1,1,1") -> (Pose) {
    alphaInputType=Bool
    rotationMode=BMM_Additive
    rotationSpace=BCS_BoneSpace
    alphaBoolBlend {blendInTime=0.5, blendOption=Linear, blendOutTime=0.5}
    boneToModify {boneName=base_back_coat_00_c}
    rotation {pitch=10}
   }
   n11 = "Transform (Modify) Bone - Bone: base_hipBag_b_r"!(ComponentPose=@n10, bAlphaBoolEnabled:bool=True, Translation="0,0,0", Rotation="-5,0,0", Scale="1,1,1") -> (Pose) {
    alphaInputType=Bool
    rotationMode=BMM_Additive
    rotationSpace=BCS_BoneSpace
    scaleMode=BMM_Additive
    scaleSpace=BCS_BoneSpace
    translationMode=BMM_Additive
    translationSpace=BCS_BoneSpace
    alphaBoolBlend {blendInTime=0.2, blendOption=Linear, blendOutTime=0.2}
    boneToModify {boneName=base_hipBag_b_r}
    rotation {pitch=-5}
   }
   n12 = "Transform (Modify) Bone - Bone: base_hipBag_a_r"!(ComponentPose=@n11, bAlphaBoolEnabled:bool=True, Translation="-0.004,0.058,-0.911", Rotation="-8,0,0", Scale="1,1,1") -> (Pose) {
    alphaInputType=Bool
    rotationMode=BMM_Additive
    rotationSpace=BCS_BoneSpace
    scaleMode=BMM_Additive
    scaleSpace=BCS_BoneSpace
    translationMode=BMM_Additive
    translationSpace=BCS_BoneSpace
    alphaBoolBlend {blendInTime=0.2, blendOption=Linear, blendOutTime=0.2}
    boneToModify {boneName=base_hipBag_a_r}
    rotation {pitch=-8}
    translation=<-0.004 0.058 -0.911>
   }
   n13 = "Transform (Modify) Bone - Bone: base_hipBag_a_l"!(ComponentPose=@n12, bAlphaBoolEnabled:bool=True, Translation="0,0,0", Rotation="3,0,-15", Scale="1,1,1") -> (Pose) {
    alphaInputType=Bool
    rotationMode=BMM_Additive
    rotationSpace=BCS_BoneSpace
    scaleMode=BMM_Additive
    scaleSpace=BCS_BoneSpace
    translationMode=BMM_Additive
    translationSpace=BCS_BoneSpace
    alphaBoolBlend {blendInTime=0.2, blendOption=Linear, blendOutTime=0.2}
    boneToModify {boneName=base_hipBag_a_l}
    rotation {pitch=3, roll=-15}
   }
   n14 = "Transform (Modify) Bone - Bone: base_hipBag_b_l"!(ComponentPose=@n13, bAlphaBoolEnabled:bool=True, Translation="0,0,0", Rotation="8,0,0", Scale="1,1,1") -> (Pose) {
    alphaInputType=Bool
    rotationMode=BMM_Additive
    rotationSpace=BCS_BoneSpace
    scaleMode=BMM_Additive
    scaleSpace=BCS_BoneSpace
    translationMode=BMM_Additive
    translationSpace=BCS_BoneSpace
    alphaBoolBlend {blendInTime=0.2, blendOption=Linear, blendOutTime=0.2}
    boneToModify {boneName=base_hipBag_b_l}
    rotation {pitch=8}
   }
   n15 = "Transform (Modify) Bone - Bone: base_coat_support_l"!(ComponentPose=@n14, bAlphaBoolEnabled:bool=True, Translation="0,0,0", Rotation="10,0,-10", Scale="1,1,1") -> (Pose) {
    alphaInputType=Bool
    rotationMode=BMM_Additive
    rotationSpace=BCS_BoneSpace
    translationMode=BMM_Additive
    translationSpace=BCS_BoneSpace
    alphaBoolBlend {blendOption=Linear}
    boneToModify {boneName=base_coat_support_l}
    rotation {pitch=10, roll=-10}
   }
   n16 = "Transform (Modify) Bone - Bone: base_coat_support_r"!(ComponentPose=@n15, bAlphaBoolEnabled:bool=True, Translation="0,0,10", Rotation="0,0,0", Scale="1,1,1") -> (Pose) {
    alphaInputType=Bool
    rotationMode=BMM_Additive
    rotationSpace=BCS_BoneSpace
    translationMode=BMM_Additive
    translationSpace=BCS_BoneSpace
    alphaBoolBlend {blendInTime=0.2, blendOption=Linear, blendOutTime=0.2}
    boneToModify {boneName=base_coat_support_r}
    translation {z=10}
   }
   n17 = "Transform (Modify) Bone - Bone: base_coat_support_l"(ComponentPose=@n16, bAlphaBoolEnabled=True, Translation="0,0,-10", Rotation="0,0,0", Scale="1,1,1") {
    alphaInputType=Bool
    rotationMode=BMM_Additive
    rotationSpace=BCS_BoneSpace
    translationMode=BMM_Additive
    translationSpace=BCS_BoneSpace
    alphaBoolBlend {blendInTime=0.2, blendOption=Linear, blendOutTime=0.2}
    boneToModify {boneName=base_coat_support_l}
    translation {z=-10}
   }
   n18 = "Transform (Modify) Bone - Bone: base_hipBag_a_l"(ComponentPose=@n17, bAlphaBoolEnabled=True, Translation="0,0,0", Rotation="25,0,-25", Scale="1,1,1") {
    alphaInputType=Bool
    rotationMode=BMM_Additive
    rotationSpace=BCS_BoneSpace
    translationMode=BMM_Additive
    translationSpace=BCS_BoneSpace
    alphaBoolBlend {blendInTime=0.2, blendOption=Linear, blendOutTime=0.2}
    boneToModify {boneName=base_hipBag_a_l}
    rotation {pitch=25, roll=-25}
   }
   n19 = "Transform (Modify) Bone - Bone: base_hipBag_a_r"(ComponentPose=@n18, bAlphaBoolEnabled=True, Translation="0,0,0", Rotation="-25,0,25", Scale="1,1,1") {
    alphaInputType=Bool
    rotationMode=BMM_Additive
    rotationSpace=BCS_BoneSpace
    translationMode=BMM_Additive
    translationSpace=BCS_BoneSpace
    alphaBoolBlend {blendInTime=0.2, blendOption=Linear, blendOutTime=0.2}
    boneToModify {boneName=base_hipBag_a_r}
    rotation {pitch=-25, roll=25}
   }
   Output Pose(Result=Component To Local(ComponentPose=@n19)) {layerGroup=None}
   n20 = Control Rig!(
    Source=Copy Pose From Mesh() {bUseAttachedParent=true}
    Alpha
    isSliding:bool=Get bIsSliding()
    isUsingPickaxe:bool=Get isPickAxeEnabled() -> (isPickAxeEnabled:bool)
    isCrouchingAlpha:double=Get isCrouchingAlpha() -> (isCrouchingAlpha:double)
    isCrouching:bool=Get bIsCrouching()
    isClothEnabled:bool=Get bClothEnabled()
    isFrontEnd:bool=Get bIsInFrontEnd()
    isPlayingEmote:double=Get isPlayingEmote() -> (isPlayingEmote:double)
    FEnoEmot:bool=Get Is FE() -> (Is FE:bool)
   ) -> (Pose) {
    alphaBoolBlend {blendOption=Linear}
    controlRigAssetReference {blueprintRigClass=&/M_MED_PreciseRuffian_CtrlRig.M_MED_PreciseRuffian_CtrlRig_C}
   }
   Save cached pose 'Start Cache'(Pose=@n20) {cacheName=Start Cache}
   n21 = Get BwdMovement() -> (BwdMovement:double)
   n22 = Trail controller - Bone: dyn_front_coat_04_l(ComponentPose=Local To Component(LocalPose=Use cached pose 'frontCloth_Cache'()), Alpha=@n21) {
    bUsePlanarLimit=true
    chainLength=3
    debugLifeTime=0.1
    maxDeltaTime=0.333
    alphaScaleBiasClamp {bInterpResult=true}
    rotationLimits [{limitMax=<180 180 180>}, {limitMax=<180 180 180>}, {limitMax=<180 180 180>}]
    rotationOffsets [<0 0 0>, <0 0 0>, <0 0 0>]
    trailBone {boneName=dyn_front_coat_04_l}
    trailRelaxationSpeed {editorCurveData {keys [<at -0.0134 19.3841 RCIM_Linear>,<at 0.5079 12.8986 RCIM_Linear>,<at 0.9955 9.6044 RCIM_Linear>]}}
   }
   n23 = Trail controller - Bone: dyn_front_coat_04_r(ComponentPose=@n22, Alpha=@n21) {
    bUsePlanarLimit=true
    chainLength=3
    debugLifeTime=0.1
    maxDeltaTime=0.333
    alphaScaleBiasClamp {bInterpResult=true}
    rotationLimits [{limitMax=<180 180 180>}, {limitMax=<180 180 180>}, {limitMax=<180 180 180>}]
    rotationOffsets [<0 0 0>, <0 0 0>, <0 0 0>]
    trailBone {boneName=dyn_front_coat_04_r}
    trailRelaxationSpeed {editorCurveData {keys [<at -0.0134 19.3841 RCIM_Linear>,<at 0.5079 12.8986 RCIM_Linear>,<at 0.9955 9.6044 RCIM_Linear>]}}
   }
   n24 = Trail controller - Bone: dyn_side_coat_04_l(ComponentPose=@n23, Alpha=@n21) {
    bUsePlanarLimit=true
    chainLength=3
    debugLifeTime=0.1
    maxDeltaTime=0.0333
    alphaScaleBiasClamp {bInterpResult=true}
    rotationLimits [{limitMax=<180 180 180>}, {limitMax=<180 180 180>}, {limitMax=<180 180 180>}]
    rotationOffsets [<0 0 0>, <0 0 0>, <0 0 0>]
    trailBone {boneName=dyn_side_coat_04_l}
    trailRelaxationSpeed {editorCurveData {keys [<at 0 15 RCIM_Linear>,<at 0.5289 13.1838 RCIM_Linear>,<at 1 6 RCIM_Linear>]}}
   }
   n25 = Trail controller - Bone: dyn_side_coat_04_r(ComponentPose=@n24, Alpha=@n21) {
    bUsePlanarLimit=true
    chainLength=3
    debugLifeTime=0.1
    maxDeltaTime=0.333
    alphaScaleBiasClamp {bInterpResult=true}
    rotationLimits [{limitMax=<180 180 180>}, {limitMax=<180 180 180>}, {limitMax=<180 180 180>}]
    rotationOffsets [<0 0 0>, <0 0 0>, <0 0 0>]
    trailBone {boneName=dyn_side_coat_04_r}
    trailRelaxationSpeed {editorCurveData {keys [<at -0.0134 19.3841 RCIM_Linear>,<at 0.5079 12.8986 RCIM_Linear>,<at 0.9955 9.6044 RCIM_Linear>]}}
   }
   n26 = Trail controller - Bone: dyn_back_coat_03_c(ComponentPose=@n25, Alpha=@n21) {
    bUsePlanarLimit=true
    chainLength=3
    debugLifeTime=0.1
    maxDeltaTime=0.0333
    alphaScaleBiasClamp {bInterpResult=true}
    rotationLimits [{limitMax=<180 180 180>}, {limitMax=<180 180 180>}, {limitMax=<180 90 180>, limitMin=<-180 -90 -180>}]
    rotationOffsets [<0 0 0>, <0 0 0>, <0 0 0>]
    trailBone {boneName=dyn_back_coat_03_c}
    trailRelaxationSpeed {editorCurveData {keys [<at 0 15 RCIM_Linear>,<at 0.5289 13.1838 RCIM_Linear>,<at 1 6 RCIM_Linear>]}}
   }
   Save cached pose 'trailCache'(Pose=Component To Local(ComponentPose=@n26)) {cacheName=trailCache}
   n27 = Trail controller - Bone: dyn_front_coat_04_l!(
    ComponentPose=Local To Component(LocalPose=Use cached pose 'frontCloth_Cache'())
    Alpha=1
    RelaxationSpeedScale:float=2
    FakeVelocity:Vector="0,0,0"
   ) -> (Pose) {
    bShowTrailLocation=true
    bUsePlanarLimit=true
    chainLength=3
    debugLifeTime=0.1
    lastBoneRotationAnimAlphaBlend=1
    maxDeltaTime=0.0333
    relaxationSpeedScale=1.2
    alphaScaleBiasClamp {bInterpResult=true, interpSpeedDecreasing=25, interpSpeedIncreasing=25}
    rotationLimits [{limitMax=<180 180 180>}, {limitMax=<180 180 180>}, {limitMax=<180 180 180>}]
    rotationOffsets [<0 0 0>, <0 0 0>, <0 0 0>]
    trailBone {boneName=dyn_front_coat_04_l}
    trailRelaxationSpeed {editorCurveData {keys [<at 0 15 RCIM_Linear>,<at 1 6 RCIM_Linear>]}}
   }
   n28 = Trail controller - Bone: dyn_front_coat_04_r!(ComponentPose=@n27, Alpha=1, RelaxationSpeedScale:float=2, FakeVelocity:Vector="0,0,0") -> (Pose) {
    bShowTrailLocation=true
    bUsePlanarLimit=true
    chainLength=3
    debugLifeTime=0.1
    lastBoneRotationAnimAlphaBlend=1
    maxDeltaTime=0.0333
    relaxationSpeedScale=1.2
    alphaScaleBiasClamp {bInterpResult=true, interpSpeedDecreasing=25, interpSpeedIncreasing=25}
    rotationLimits [{limitMax=<180 180 180>}, {limitMax=<180 180 180>}, {limitMax=<180 180 180>}]
    rotationOffsets [<0 0 0>, <0 0 0>, <0 0 0>]
    trailBone {boneName=dyn_front_coat_04_r}
    trailRelaxationSpeed {editorCurveData {keys [<at 0 15 RCIM_Linear>,<at 1 6 RCIM_Linear>]}}
   }
   n29 = Trail controller - Bone: dyn_side_coat_04_l!(ComponentPose=@n28, Alpha=1, RelaxationSpeedScale:float=2, FakeVelocity:Vector="0,0,0") -> (Pose) {
    bReorientParentToChild=false
    bShowTrailLocation=true
    bUsePlanarLimit=true
    chainLength=3
    debugLifeTime=0.1
    lastBoneRotationAnimAlphaBlend=1
    maxDeltaTime=0.0333
    relaxationSpeedScale=1.2
    alphaScaleBiasClamp {bInterpResult=true, interpSpeedDecreasing=25, interpSpeedIncreasing=25}
    rotationLimits [{limitMax=<180 180 180>}, {limitMax=<180 180 180>}, {limitMax=<180 180 180>}]
    rotationOffsets [<0 0 0>, <0 0 0>, <0 0 0>]
    trailBone {boneName=dyn_side_coat_04_l}
    trailRelaxationSpeed {editorCurveData {keys [<at 0 15 RCIM_Linear>,<at 1 6 RCIM_Linear>]}}
   }
   n30 = Trail controller - Bone: dyn_side_coat_04_r!(ComponentPose=@n29, Alpha=1, RelaxationSpeedScale:float=2, FakeVelocity:Vector="0,0,0") -> (Pose) {
    bShowTrailLocation=true
    bUsePlanarLimit=true
    chainLength=3
    debugLifeTime=0.1
    lastBoneRotationAnimAlphaBlend=1
    maxDeltaTime=0.0333
    relaxationSpeedScale=1.2
    alphaScaleBiasClamp {bInterpResult=true, interpSpeedDecreasing=25, interpSpeedIncreasing=25}
    rotationLimits [{limitMax=<180 180 180>}, {limitMax=<180 180 180>}, {limitMax=<180 180 180>}]
    rotationOffsets [<0 0 0>, <0 0 0>, <0 0 0>]
    trailBone {boneName=dyn_side_coat_04_r}
    trailRelaxationSpeed {editorCurveData {keys [<at 0 15 RCIM_Linear>,<at 1 6 RCIM_Linear>]}}
   }
   n31 = Trail controller - Bone: dyn_back_coat_03_c!(ComponentPose=@n30, Alpha=1, RelaxationSpeedScale:float=2, FakeVelocity:Vector="0,0,0") -> (Pose) {
    bShowTrailLocation=true
    bUsePlanarLimit=true
    chainLength=3
    debugLifeTime=0.1
    lastBoneRotationAnimAlphaBlend=1
    maxDeltaTime=0.0333
    relaxationSpeedScale=1.2
    alphaScaleBiasClamp {bInterpResult=true, interpSpeedDecreasing=25, interpSpeedIncreasing=25}
    rotationLimits [{limitMax=<180 180 180>}, {limitMax=<180 180 180>}, {limitMax=<180 180 180>}]
    rotationOffsets [<0 0 0>, <0 0 0>, <0 0 0>]
    trailBone {boneName=dyn_back_coat_03_c}
    trailRelaxationSpeed {editorCurveData {keys [<at 0 15 RCIM_Linear>,<at 1 6 RCIM_Linear>]}}
   }
   Save cached pose 'dbnoCache'(Pose=Component To Local(ComponentPose=@n31)) {cacheName=dbnoCache}
   n32 = Anim Dynamics - Bone: dyn_sleeveFore_c_l!(
    ComponentPose=Local To Component(LocalPose=Use cached pose 'Start Cache'())
    Alpha=1
    GravityOverride:Vector="0,0,-980"
    ExternalForce="0,0,0"
   ) -> (Pose) {
    angularDampingOverride=1
    angularSpringConstant=20000
    bAngularSpring=true
    bOverrideAngularDamping=true
    bUseGravityOverride=true
    gravityScale=0.35
    simulationSpace=BoneRelative
    alphaBoolBlend {blendOption=Linear}
    boundBone {boneName=dyn_sleeveFore_c_l}
    componentLinearAccScale=<1 1 1>
    componentLinearVelScale=<1 1 1>
    gravityOverride {z=-980}
    physicsBodyDefinitions [{
      boundBone {boneName=dyn_sleeveFore_c_l}
      boxExtents=<20 30 30>
      constraintSetup {angularLimitsMax=<15 15 15>, angularLimitsMin=<-15 -15 -15>, angularTarget=<1 0 0>}
      localJointOffset=<4 0 0>
     }]
    relativeSpaceBone {boneName=upperarm_l}
   }
   n33 = Anim Dynamics - Bone: dyn_sleeveFore_a_l!(ComponentPose=@n32, Alpha=1, GravityOverride:Vector="0,0,-980", ExternalForce="0,0,0") -> (Pose) {
    angularDampingOverride=1
    angularSpringConstant=20000
    bAngularSpring=true
    bOverrideAngularDamping=true
    bUseGravityOverride=true
    gravityScale=0.35
    simulationSpace=BoneRelative
    alphaBoolBlend {blendOption=Linear}
    boundBone {boneName=dyn_sleeveFore_a_l}
    componentLinearAccScale=<1 1 1>
    componentLinearVelScale=<1 1 1>
    gravityOverride {z=-980}
    physicsBodyDefinitions [{
      boundBone {boneName=dyn_sleeveFore_a_l}
      boxExtents=<20 30 30>
      constraintSetup {angularLimitsMax=<15 15 15>, angularLimitsMin=<-15 -15 -15>, angularTarget=<1 0 0>}
      localJointOffset=<4 0 0>
     }]
    relativeSpaceBone {boneName=upperarm_l}
   }
   n34 = Anim Dynamics - Bone: dyn_sleeveFore_b_l!(ComponentPose=@n33, Alpha=1, GravityOverride:Vector="0,0,-980", ExternalForce="0,0,0") -> (Pose) {
    angularDampingOverride=1
    angularSpringConstant=20000
    bAngularSpring=true
    bOverrideAngularDamping=true
    bUseGravityOverride=true
    gravityScale=0.35
    simulationSpace=BoneRelative
    alphaBoolBlend {blendOption=Linear}
    boundBone {boneName=dyn_sleeveFore_b_l}
    componentLinearAccScale=<1 1 1>
    componentLinearVelScale=<1 1 1>
    gravityOverride {z=-980}
    physicsBodyDefinitions [{
      boundBone {boneName=dyn_sleeveFore_b_l}
      boxExtents=<20 30 30>
      constraintSetup {angularLimitsMax=<15 15 15>, angularLimitsMin=<-15 -15 -15>, angularTarget=<1 0 0>}
      localJointOffset=<4 0 0>
     }]
    relativeSpaceBone {boneName=upperarm_l}
   }
   n35 = Anim Dynamics - Bone: dyn_sleeveFore_a_r!(ComponentPose=@n34, Alpha=1, GravityOverride:Vector="0,0,-980", ExternalForce="0,0,0") -> (Pose) {
    angularDampingOverride=1
    angularSpringConstant=20000
    bAngularSpring=true
    bOverrideAngularDamping=true
    bUseGravityOverride=true
    gravityScale=0.35
    simulationSpace=BoneRelative
    alphaBoolBlend {blendOption=Linear}
    boundBone {boneName=dyn_sleeveFore_a_r}
    componentLinearAccScale=<1 1 1>
    componentLinearVelScale=<1 1 1>
    gravityOverride {z=-980}
    physicsBodyDefinitions [{
      boundBone {boneName=dyn_sleeveFore_a_r}
      boxExtents=<20 30 30>
      constraintSetup {angularLimitsMax=<15 15 15>, angularLimitsMin=<-15 -15 -15>, angularTarget=<1 0 0>}
      localJointOffset=<4 0 0>
     }]
    relativeSpaceBone {boneName=upperarm_r}
   }
   n36 = Anim Dynamics - Bone: dyn_sleeveFore_c_r!(ComponentPose=@n35, Alpha=1, GravityOverride:Vector="0,0,-980", ExternalForce="0,0,0") -> (Pose) {
    angularDampingOverride=1
    angularSpringConstant=20000
    bAngularSpring=true
    bOverrideAngularDamping=true
    bUseGravityOverride=true
    gravityScale=0.35
    simulationSpace=BoneRelative
    alphaBoolBlend {blendOption=Linear}
    boundBone {boneName=dyn_sleeveFore_c_r}
    componentLinearAccScale=<1 1 1>
    componentLinearVelScale=<1 1 1>
    gravityOverride {z=-980}
    physicsBodyDefinitions [{
      boundBone {boneName=dyn_sleeveFore_c_r}
      boxExtents=<20 30 30>
      constraintSetup {angularLimitsMax=<15 15 15>, angularLimitsMin=<-15 -15 -15>, angularTarget=<1 0 0>}
      localJointOffset=<4 0 0>
     }]
    relativeSpaceBone {boneName=upperarm_r}
   }
   n37 = Anim Dynamics - Bone: dyn_sleeveFore_b_r!(ComponentPose=@n36, Alpha=1, GravityOverride:Vector="0,0,-980", ExternalForce="0,0,0") -> (Pose) {
    angularDampingOverride=1
    angularSpringConstant=20000
    bAngularSpring=true
    bOverrideAngularDamping=true
    bUseGravityOverride=true
    gravityScale=0.35
    simulationSpace=BoneRelative
    alphaBoolBlend {blendOption=Linear}
    boundBone {boneName=dyn_sleeveFore_b_r}
    componentLinearAccScale=<1 1 1>
    componentLinearVelScale=<1 1 1>
    gravityOverride {z=-980}
    physicsBodyDefinitions [{
      boundBone {boneName=dyn_sleeveFore_b_r}
      boxExtents=<20 30 30>
      constraintSetup {angularLimitsMax=<15 15 15>, angularLimitsMin=<-15 -15 -15>, angularTarget=<1 0 0>}
      localJointOffset=<4 0 0>
     }]
    relativeSpaceBone {boneName=upperarm_r}
   }
   Save cached pose 'SleeveDynamics Cache'(Pose=Component To Local(ComponentPose=@n37)) {cacheName=SleeveDynamics Cache}
   n38 = RigidBody!(
    ComponentPose=Local To Component(LocalPose=Use cached pose 'SleeveDynamics Cache'())
    Alpha=Get RB_Coat_alpha() -> (RB_Coat_alpha:double)
    OverrideWorldGravity:Vector=Get GravityOverride()
    ExternalForce
   ) -> (Pose) {
    bClampLinearTranslationLimitToRefPose=true
    bOverrideWorldGravity=true
    bUseExternalClothCollision=true
    overridePhysicsAsset=&/M_MED_PreciseRuffian_Physics
    simulationSpace=BaseBoneSpace
    alphaBoolBlend {blendOption=Linear}
    baseBoneRef {boneName=pelvis}
    componentAppliedLinearAccClamp=<3000 3000 3000>
    componentLinearAccScale=<0.3 0.3 0.3>
    componentLinearVelScale=<0.3 0.3 0.3>
    simSpaceSettings {maxAngularAcceleration=1000, maxAngularVelocity=3, maxLinearAcceleration=700, maxLinearVelocity=500, velocityScaleZ=0.2, worldAlpha=0.125}
   }
   n39 = "Transform (Modify) Bone - Bone: base_front_coat_01_r"(ComponentPose=@n38, bAlphaBoolEnabled=True, Translation="0,0,0", Rotation="0,0,10", Scale="1,1,1") {
    alphaInputType=Bool
    rotationMode=BMM_Additive
    rotationSpace=BCS_BoneSpace
    translationMode=BMM_Additive
    translationSpace=BCS_BoneSpace
    alphaBoolBlend {blendInTime=0.2, blendOption=Linear, blendOutTime=0.2}
    boneToModify {boneName=base_front_coat_01_r}
    rotation {roll=10}
   }
   Save cached pose 'physicsCoat_Cache'(Pose=Component To Local(ComponentPose=@n39)) {cacheName=physicsCoat_Cache}
   n40 = RigidBody(ComponentPose=Local To Component(LocalPose=Use cached pose 'SleeveDynamics Cache'()), OverrideWorldGravity=Get GravityOverride()) {
    bClampLinearTranslationLimitToRefPose=true
    bOverrideWorldGravity=true
    bUseExternalClothCollision=true
    overridePhysicsAsset=&/M_MED_PreciseRuffian_Physics_FrontCloth
    simulationSpace=BaseBoneSpace
    alphaBoolBlend {blendOption=Linear}
    baseBoneRef {boneName=pelvis}
    componentAppliedLinearAccClamp=<3000 3000 3000>
    componentLinearAccScale=<0.3 0.3 0.3>
    componentLinearVelScale=<0.3 0.3 0.3>
    simSpaceSettings {maxAngularAcceleration=1000, maxAngularVelocity=3, maxLinearAcceleration=700, maxLinearVelocity=500, velocityScaleZ=0.2, worldAlpha=0.125}
   }
   Save cached pose 'frontCloth_Cache'(Pose=Component To Local(ComponentPose=@n40)) {cacheName=frontCloth_Cache}
  ]}
 EventGraph {nodes [
   n1 = Event BlueprintUpdateAnimation()
   > Set DeltaTimeLocal(DeltaTimeLocal:double=@n1.DeltaTimeX) -> (Output_Get:double)
   > n2 = Sequence!(execute) -> (then_0, then_1, then_2:Exec, then_3:Exec, then_4:Exec)
   n3 = Get bClothEnabled()
   > n4 = Branch(Condition=@n3)
   > n5 = Sequence!(execute) -> (then_0, then_1)
   n6 = Gravity Override State Function DEPRECATED(
    execute=[@n5.then_0, @n4.else]
    CharacterPartAnimInstance=Self-Reference()
    GameTimeInSeconds=GetGameTimeInSeconds()
    GravityOverrideParams=Get Gravity Override Params()
   )
   > n7 = Set GravityOverride(GravityOverride=@n6)
   n8 = AND Boolean(A=Get bIsCrouching(), B=NOT Boolean(A=Get bIsMoving2D()))
   > Set isCrouchingAlpha(isCrouchingAlpha:double=FInterpTo(
     Current=Get isCrouchingAlpha()
     Target="To Float (Boolean)"(InBool=@n8)
     DeltaTime=Get DeltaTimeLocal() -> (DeltaTimeLocal:double)
     InterpSpeed=2
    )) -> (Output_Get:double)
   n9 = Get bIsPlayingEmote()
   > Set isPlayingEmote(isPlayingEmote:double=FInterpTo(Current=Get isPlayingEmote(), Target="To Float (Boolean)"(InBool=@n9), DeltaTime=Get DeltaTimeLocal(), InterpSpeed=2)) -> (Output_Get:double)
   > Set skydivingDive(skydivingDive:bool=OR Boolean(A=Get bIsSkydiveDiveMode(), B=Get bIsSkydiveFloating())) -> (Output_Get:bool)
   n10 = Get bIsPlayingEmote()
   > Set Is FE(Is FE:bool=AND Boolean(A=Get bIsInFrontEnd(), B=NOT Boolean(A=@n10))) -> (Output_Get:bool)
   n11 = Cloth Controller Function DEPRECATED!(
    execute=@n5.then_1
    CharacterPartAnimInstance:Custom Character Part Anim Instance Object Reference=Self-Reference()
    ClothingAssetName:String=PreciseRuffian_Cloth
    Cloth Paramaters:Cloth Params Struct Structure=Get Cloth Paramaters_0() -> (Cloth Paramaters_0:Cloth Params Struct Structure)
    Cloth Core Settings:Cloth Core Setting Struct Structure=Get Cloth Core Settings_0() -> (Cloth Core Settings_0:Cloth Core Setting Struct Structure)
    ClothUpAxis:Vector="1,0,0"
    OnGround_Moving_SpeedLimit:double=2
    OnGround_Moving_AngularVel_Max:double=0.05
    OnGround_Moving_FictitiousVel_Max:double=0.6
    LinVelScale_FwdBwd_Clamp:double=0
    LinVelScale_UpDwn_Clamp:double=0.7
    DeltaTime:double=Get DeltaTimeLocal()
    GameTimeInSeconds:double=GetGameTimeInSeconds()
    WindGustsParams:Wind Gust Params Struct Structure=Get Wind Gust Params() -> (Wind Gust Params:Wind Gust Params Struct Structure)
    WindGustLocalSettings:Wind Gust Local Setting Struct Structure=Get Gust Local Settings() -> (Gust Local Settings:Wind Gust Local Setting Struct Structure)
   ) -> (
    then
    CurrentStateOut:String
    Cloth Settings Out:Cloth Setting Struct Structure
    Cloth Core Settings Out:Cloth Core Setting Struct Structure
    WindGustSettings Out:Wind Gust Setting Struct Structure
    WindGustLocalSettings Out:Wind Gust Local Setting Struct Structure
    GustStrength - WindVelocityOut:Vector
    LinearVelVector-DELETE LATER:Vector
   )
   > Set Gust Local Settings(Gust Local Settings:Wind Gust Local Setting Struct Structure=@n11.WindGustLocalSettings Out) -> (Output_Get:Wind Gust Local Setting Struct Structure)
   n12 = Get bIsMoving2D()
   > Set isClothEnabledCrouch(isClothEnabledCrouch:bool=AND Boolean!(A=Get bIsCrouching(), B=Get bClothEnabled(), C:bool=NOT Boolean(A=@n12)) -> (ReturnValue)) -> (Output_Get:bool)
   n13 = PlayerMovement(execute=@n2.then_1, AnimInstance=Self-Reference(), XDotProductCoefficient=100, YDotProductCoefficient=100, ZDotProductCoefficient=100)
   n14 = BreakVector(InVec=MakeVector(X=@n13, Y=@n13.YMovement))
   n15 = "To Float (Boolean)"(InBool="float < float"(A=@n14, B=-50))
   > Set LeftMovement(LeftMovement:double=@n15) -> (Output_Get:double)
   n16 = "To Float (Boolean)"(InBool="float > float"(A=@n14, B=50))
   > Set RightMovement(RightMovement:double=@n16) -> (Output_Get:double)
   n17 = "To Float (Boolean)"(InBool="float > float"(A=@n14.Y, B=40))
   n18 = float + float!(A=@n15, B=@n16, C:double="To Float (Boolean)"(InBool="float < float"(A=@n14.Y, B=-40))) -> (ReturnValue)
   > n19 = Set FwdMovement(FwdMovement:double="Clamp (Float)"(Value=float - float(A=@n17, B=@n18), Min=0, Max=0.7)) -> (Output_Get:double)
   n20 = Branch(execute=@n2.then_3)
   > Set TrailControllerRot(TrailControllerRot:Rotator="15,0, 0") -> (Output_Get:Rotator)
   n21 = Branch(execute=@n20.else, Condition=Get bIsSkyDiving())
   > n22 = Branch(Condition=Get bIsParachuteOpened())
   > Set TrailControllerRot(TrailControllerRot="45,0, 0")
   n23 = Branch(execute=@n22.else, Condition="float > float"(A=Get PawnVelocityZ(), B=50) {is=title K2Node_CallFunction})
   > Set BwdMovement(BwdMovement:double=float * float(A=Get BwdMovement(), B=0)) -> (Output_Get:double)
   > Set TrailControllerRot(TrailControllerRot="45,0, 0")
   n24 = Branch(execute=@n23.else, Condition=Get bIsSkydiveDiveMode())
   > Set TrailControllerRot(TrailControllerRot="15,0, 0")
   Set TrailControllerRot(execute=@n24.else, TrailControllerRot="60,0, 0")
   Set TrailControllerRot(execute=@n21.else, TrailControllerRot="0, 0, 0")
   > Set BwdMovement(BwdMovement=float * float(A=Get BwdMovement(), B=1))
   n25 = CheckForBackbling()
   Set isWearingBackbling(execute=@n2.then_4, isWearingBackbling=AND Boolean(A=@n25, B=NOT Boolean(A=@n25.bIsBackblingACape), C=NOT Boolean(A=@n25.bIsBackblingLong)))
   > n26 = Set Is Wearing Cape(Is Wearing Cape=OR Boolean(A=@n25.bIsBackblingACape, B=@n25.bIsBackblingLong))
   > n27 = Branch(Condition=@n26)
   > n28 = Compare Float!(Exec:Exec, Input:double=Get BwdMovement(), CompareWith:double) -> (">":Exec, "==":Exec, "<":Exec) {is=title K2Node_MacroInstance}
   Set RB_Coat_alpha(execute=[@n27.else, @n28."<", @n28."=="], RB_Coat_alpha:double=1) -> (Output_Get:double)
   Set RB_Coat_alpha(execute=@n28.">", RB_Coat_alpha=0.4)
   n29 = PlayerMovement(execute=@n2.then_2, AnimInstance=Self-Reference(), XDotProductCoefficient=1, YDotProductCoefficient=1, ZDotProductCoefficient=1)
   n30 = MapRangeClamped(Value=@n29.YMovement, InRangeA=0, InRangeB=1)
   n31 = "Absolute (Float)"(A=@n29)
   > Set BwdMovement(BwdMovement=MapRangeClamped(Value=float + float(A=@n31, B=@n30) {is=title K2Node_CommutativeAssociativeBinaryOperator}, InRangeA=0, InRangeB=1.4))
   n32 = Event OnCharacterCustomizationLoaded()
   > Set SkelMesh(SkelMesh=@n32.CustomBaseMesh)
   Event OnWeaponChanged()
   > n33 = Cast To FortPlayerAnimInstance(Object=GetAnimInstance(self=Get SkelMesh()))
   > Set isPickAxeEnabled(isPickAxeEnabled:bool=false) -> (Output_Get:bool)
   > Switch on EFortWeaponCoreAnimation!(execute, Selection:EFortWeaponCoreAnimation Enum=Get WeaponCoreAnim(self=@n33)) -> (
    Melee:Exec
    Pistol:Exec
    Shotgun:Exec
    PaperBlueprint:Exec
    Rifle:Exec
    MeleeOneHand:Exec
    MachinePistol:Exec
    RocketLauncher:Exec
    GrenadeLauncher:Exec
    GoingCommando:Exec
    AssaultRifle:Exec
    TacticalShotgun:Exec
    SniperRifle:Exec
    TrapPlacement:Exec
    ShoulderLauncher:Exec
    AbilityDecoTool:Exec
    Crossbow:Exec
    C4:Exec
    RemoteControl:Exec
    DualWield:Exec
    AR_BullPup:Exec
    AR_ForwardGrip:Exec
    MedPackPaddles:Exec
    SMG_P90:Exec
    AR_DrumGun:Exec
    Consumable_Small:Exec
    Consumable_Large:Exec
    Balloon:Exec
    MountedTurret:Exec
    CreativeTool:Exec
    ExplosiveBow:Exec
    AshtonIndigo:Exec
    AshtonChicago:Exec
    MeleeDualWield:Exec
    Unarmed:Exec
   )
   > Set isPickAxeEnabled(isPickAxeEnabled=true)
   Event OnBeginSkydiving()
   > Set TrailControllerRot(TrailControllerRot="-65,0,0")
   Event OnEndSkydiving()
   > Set TrailControllerRot(TrailControllerRot="-10,0,0")
   Break Cloth Setting Struct!(ClothSettingStruct:"Cloth Setting Struct Structure (by ref)"=@n11.Cloth Settings Out) -> (
    LinearVelocityScale_7_#1:Vector=~
    AngularVelocityScale_6_#2:float=~
    FictitiousAngularScale_10_#3:float=~
    GravityOverride_15_#4:Vector=~
    DragCoefficient_23_#5:Vector 2D Structure
    LiftCoefficient_24_#6:Vector 2D Structure
    JointName_27_#7:Name
   ) {is=title K2Node_BreakStruct}
   DrawVectorFromJointFunction!(
    execute
    SkelMeshComponent:Scene Component Object Reference=GetPartSkeletalMeshComponent()
    JointName:Name=pelvis
    LineEnd:Vector=@n7
    JointOffset:Vector="0, 0, 0"
    LineColor:Linear Color Structure=(R=1,G=0,B=0,A=1)
    Thickness:double=2
    Foreground:EDrawDebugSceneDepthPriorityGroup Enum=Foreground
   ) -> (then)
   OR Boolean(A=@n3, B=NOT Boolean(A=Get bIsPlayingEmote()))
   n34 = Get DeltaTimeLocal()
   n35 = Get FwdMovement() -> (FwdMovement:double)
   n36 = Get BwdMovement()
   Set BwdMovement(BwdMovement=float * float(A=@n19, B=FInterpTo(Current=@n36, Target=@n35, DeltaTime=@n34, InterpSpeed=8)) {is=title K2Node_PromotableOperator})
  ]}
]
variables [
 BwdMovement {type=number}
 Cloth Core Settings_0 {
  type=ClothCoreSettingStruct
  default {
   collisionThickness=0
   dampingCoefficient=0.2
   frictionCoefficient=0.3
   localDampingCoefficient=0
   numberIterations=1
   numberSubsteps=2
   selfCollisionThickness=0
   animDriveDamping=<0 1>
   animDriveStiffness=<0 1>
   areaStiffness=<1 1>
   bendingStiffness=<1 1>
   edgeStiffness=<1 1>
   tetherStiffness=<1 1>
  }
 }
 Cloth Paramaters_0 {
  type=ClothParamsStruct
  default {
   dBNO {
    angularVelocityScale=0.6
    fictitiousAngularScale=0.1
    jointName=pelvis
    dragCoefficient=<0.2 1>
    gravityOverride=<-980 400 0>
    liftCoefficient=<0 0.5>
    linearVelocityScale=<0.3 0.3 0.1>
   }
   dBNO_Carried {
    angularVelocityScale=0.6
    fictitiousAngularScale=0.1
    jointName=pelvis
    dragCoefficient=<0.2 1>
    gravityOverride=<-980 -200 0>
    liftCoefficient=<0 0.5>
    linearVelocityScale=<0.3 0.3 0.1>
   }
   dBNO_Impulse {
    angularVelocityScale=0
    fictitiousAngularScale=0
    jointName=pelvis
    dragCoefficient=<0 0.2>
    gravityOverride=<-3000 -2000 0>
    liftCoefficient=<0 0.1>
    linearVelocityScale=<0.3 0.1 0.3>
   }
   dBNO_RollBwd {
    angularVelocityScale=0
    fictitiousAngularScale=0
    jointName=pelvis
    dragCoefficient=<0 0.2>
    gravityOverride=<-2000 0 0>
    liftCoefficient=<0 0.1>
    linearVelocityScale=<0.3 0.1 0.3>
   }
   dBNO_RollFwd {
    angularVelocityScale=0
    fictitiousAngularScale=0
    jointName=pelvis
    dragCoefficient=<0 0.2>
    gravityOverride=<-980 0 0>
    liftCoefficient=<0 0.1>
    linearVelocityScale=<0.3 0.1 0.3>
   }
   dBNO_Tumbling {
    angularVelocityScale=0
    fictitiousAngularScale=0
    jointName=pelvis
    dragCoefficient=<0 0.2>
    gravityOverride=<-2500 -300 0>
    liftCoefficient=<0 0.1>
    linearVelocityScale=<0.3 0.1 0.3>
   }
   emoteOrMelee {
    angularVelocityScale=0.2
    fictitiousAngularScale=0.1
    jointName=pelvis
    dragCoefficient=<0 0.3>
    gravityOverride=<-980 0 0>
    liftCoefficient=<0 0.2>
    linearVelocityScale=<0.05 0.05 0.05>
   }
   front_End {
    angularVelocityScale=0.1
    fictitiousAngularScale=0.2
    jointName=root
    dragCoefficient=<0.2 0.2>
    gravityOverride=<0 -200 -980>
    liftCoefficient=<0.1 0.1>
    linearVelocityScale=<0.1 0.1 0.1>
   }
   onGround_Crouch_Moving {
    angularVelocityScale=0.1
    fictitiousAngularScale=0.3
    jointName=root
    dragCoefficient=<2 2>
    gravityOverride=<0 0 -980>
    liftCoefficient=<0.6 0.6>
    linearVelocityScale=<0.6 0.6 0.5>
   }
   onGround_Falling {
    angularVelocityScale=0.6
    fictitiousAngularScale=0
    jointName=root
    dragCoefficient=<0 0.2>
    gravityOverride=<0 -500 -980>
    liftCoefficient=<0 0.1>
    linearVelocityScale=<2 1 0.3>
   }
   onGround_Moving {
    angularVelocityScale=0.4
    fictitiousAngularScale=0.1
    jointName=root
    dragCoefficient=<0 0.4>
    gravityOverride=<0 -100 -980>
    liftCoefficient=<0 0.2>
    linearVelocityScale=<0.25 0.3 0.3>
   }
   onGround_Standing {
    angularVelocityScale=0.05
    fictitiousAngularScale=0.1
    jointName=root
    dragCoefficient=<0 0.2>
    gravityOverride=<0 -100 -980>
    liftCoefficient=<0 0.2>
    linearVelocityScale=<0.4 0.2 0.2>
   }
   rail_Grinding {
    angularVelocityScale=0.2
    fictitiousAngularScale=0
    jointName=root
    dragCoefficient=<0.1 2>
    gravityOverride=<300 -500 -800>
    liftCoefficient=<0.1 2>
    linearVelocityScale=<0 0 0>
   }
   riding_Moving {
    angularVelocityScale=0.05
    fictitiousAngularScale=1
    jointName=root
    dragCoefficient=<0.5 0.5>
    gravityOverride=<0 0 -980>
    liftCoefficient=<0.5 0.5>
    linearVelocityScale=<0.4 0.6 0.3>
   }
   riding_Moving_Targeting {
    angularVelocityScale=0.05
    fictitiousAngularScale=0.7
    jointName=root
    dragCoefficient=<0.5 0.5>
    gravityOverride=<0 0 -980>
    liftCoefficient=<0.5 0.5>
    linearVelocityScale=<0.2 0.6 0.2>
   }
   riding_Standing {
    angularVelocityScale=0.05
    fictitiousAngularScale=1
    jointName=root
    dragCoefficient=<0.5 0.5>
    gravityOverride=<0 -450 -980>
    liftCoefficient=<0.5 0.5>
    linearVelocityScale=<0.6 0.6 0.3>
   }
   skydive_DivingDown {
    angularVelocityScale=0.1
    fictitiousAngularScale=0
    jointName=pelvis
    dragCoefficient=<0 1>
    gravityOverride=<-450 -100 0>
    liftCoefficient=<0 1>
    linearVelocityScale=<0.1 0.1 0.1>
   }
   skydive_FloatingDown {
    angularVelocityScale=0.1
    fictitiousAngularScale=0
    jointName=root
    dragCoefficient=<0 0.5>
    gravityOverride=<0 -900 900>
    liftCoefficient=<0 1>
    linearVelocityScale=<0.1 0.1 0.1>
   }
   skydive_GoingUp {
    angularVelocityScale=0.1
    fictitiousAngularScale=0
    jointName=root
    dragCoefficient=<0 0>
    gravityOverride=<0 -3000 3000>
    liftCoefficient=<0 0>
    linearVelocityScale=<0.1 0.1 0.05>
   }
   skydive_Parachuting {
    angularVelocityScale=0.1
    fictitiousAngularScale=0
    jointName=root
    dragCoefficient=<0 1>
    gravityOverride=<0 -500 -980>
    liftCoefficient=<0 1>
    linearVelocityScale=<0.05 0.05 0.1>
   }
   skydive_Surfing {
    angularVelocityScale=0.6
    fictitiousAngularScale=0.1
    jointName=root
    dragCoefficient=<1.5 1.5>
    gravityOverride=<0 -600 -500>
    liftCoefficient=<1 1>
    linearVelocityScale=<0.1 0.1 0.3>
   }
   swimming {
    angularVelocityScale=0.6
    fictitiousAngularScale=0
    jointName=root
    dragCoefficient=<0 2>
    gravityOverride=<0 -600 400>
    liftCoefficient=<0 0.4>
    linearVelocityScale=<0.6 0.2 0.2>
   }
   tactical_Sprint {
    angularVelocityScale=0.3
    fictitiousAngularScale=0.1
    jointName=root
    dragCoefficient=<0 1>
    gravityOverride=<0 0 -980>
    liftCoefficient=<0 0.75>
    linearVelocityScale=<0.25 0.3 0.3>
   }
   ziplining {
    angularVelocityScale=0.1
    fictitiousAngularScale=0
    jointName=root
    dragCoefficient=<0 0.5>
    gravityOverride=<0 0 -980>
    liftCoefficient=<0 0.3>
    linearVelocityScale=<0.1 0.1 0.1>
   }
  }
 }
 DeltaTimeLocal {type=number}
 FwdMovement {type=number}
 Gravity Override Params {type=GravityOverrideParamsStruct}
 GravityOverride {type=Vector}
 Gust Local Settings {type=WindGustLocalSettingStruct, default {elapsedTime=0, finalGustStrengthLocal=0, gustDuration=0, gustStrength=0, timeBetweenGusts=0}}
 Is FE {type=boolean}
 Is Wearing Cape {type=boolean}
 isClothEnabledCrouch {type=boolean}
 isCrouchingAlpha {type=number}
 isPickAxeEnabled {type=boolean}
 isPlayingEmote {type=number}
 isSkydivingFloating {type=number}
 isWearingBackbling {type=boolean}
 LeftMovement {type=number}
 RB_Coat_alpha {type=number}
 RightMovement {type=number}
 SkelMesh {type=SkeletalMeshComponent}
 skydivingDive {type=boolean}
 TrailControllerRot {type=Rotator}
 Wind Gust Params {
  type=WindGustParamsStruct
  default {
   dBNO {
    maxDownInterpSpeed=5
    maxGustDuration=2
    maxGustStrength=7
    maxTimeBetweenGusts=7
    maxUpInterpSpeed=5
    minDownInterpSpeed=2
    minGustDuration=0.5
    minGustStrength=3
    minTimeBetweenGusts=0
    minUpInterpSpeed=2
    windTurbulenceScalarA=0.1
    windTurbulenceScalarB=1
    windAmplitude=<0 0 0>
    windFrequency=<0 0 0>
    windOffset=<0 0 0>
   }
   dBNO_Carried {
    maxDownInterpSpeed=5
    maxGustDuration=2
    maxGustStrength=7
    maxTimeBetweenGusts=7
    maxUpInterpSpeed=5
    minDownInterpSpeed=2
    minGustDuration=0.5
    minGustStrength=3
    minTimeBetweenGusts=0
    minUpInterpSpeed=2
    windTurbulenceScalarA=0.1
    windTurbulenceScalarB=1
    windAmplitude=<0 0 0>
    windFrequency=<0 0 0>
    windOffset=<0 0 0>
   }
   dBNO_Impulse {
    maxDownInterpSpeed=5
    maxGustDuration=2
    maxGustStrength=7
    maxTimeBetweenGusts=7
    maxUpInterpSpeed=5
    minDownInterpSpeed=2
    minGustDuration=0.5
    minGustStrength=3
    minTimeBetweenGusts=0
    minUpInterpSpeed=2
    windTurbulenceScalarA=0.1
    windTurbulenceScalarB=1
    windAmplitude=<0 0 0>
    windFrequency=<0 0 0>
    windOffset=<0 0 0>
   }
   dBNO_RollBwd {
    maxDownInterpSpeed=5
    maxGustDuration=2
    maxGustStrength=7
    maxTimeBetweenGusts=7
    maxUpInterpSpeed=5
    minDownInterpSpeed=2
    minGustDuration=0.5
    minGustStrength=3
    minTimeBetweenGusts=0
    minUpInterpSpeed=2
    windTurbulenceScalarA=0.1
    windTurbulenceScalarB=1
    windAmplitude=<0 0 0>
    windFrequency=<0 0 0>
    windOffset=<0 0 0>
   }
   dBNO_RollFwd {
    maxDownInterpSpeed=5
    maxGustDuration=2
    maxGustStrength=7
    maxTimeBetweenGusts=7
    maxUpInterpSpeed=5
    minDownInterpSpeed=2
    minGustDuration=0.5
    minGustStrength=3
    minTimeBetweenGusts=0
    minUpInterpSpeed=2
    windTurbulenceScalarA=0.1
    windTurbulenceScalarB=1
    windAmplitude=<0 0 0>
    windFrequency=<0 0 0>
    windOffset=<0 0 0>
   }
   dBNO_Tumbling {
    maxDownInterpSpeed=5
    maxGustDuration=2
    maxGustStrength=7
    maxTimeBetweenGusts=7
    maxUpInterpSpeed=5
    minDownInterpSpeed=2
    minGustDuration=0.5
    minGustStrength=3
    minTimeBetweenGusts=0
    minUpInterpSpeed=2
    windTurbulenceScalarA=0.1
    windTurbulenceScalarB=1
    windAmplitude=<0 0 0>
    windFrequency=<0 0 0>
    windOffset=<0 0 0>
   }
   emoteOrMelee {
    maxDownInterpSpeed=5
    maxGustDuration=2
    maxGustStrength=7
    maxTimeBetweenGusts=7
    maxUpInterpSpeed=5
    minDownInterpSpeed=2
    minGustDuration=0.5
    minGustStrength=3
    minTimeBetweenGusts=0
    minUpInterpSpeed=2
    windTurbulenceScalarA=0.1
    windTurbulenceScalarB=1
    windAmplitude=<30 25 50>
    windFrequency=<2 3 5>
    windOffset=<-300 -150 0>
   }
   front_End {
    maxDownInterpSpeed=5
    maxGustDuration=2
    maxGustStrength=4
    maxTimeBetweenGusts=7
    maxUpInterpSpeed=5
    minDownInterpSpeed=2
    minGustDuration=0.5
    minGustStrength=2
    minTimeBetweenGusts=0
    minUpInterpSpeed=2
    windTurbulenceScalarA=0.1
    windTurbulenceScalarB=1
    windAmplitude=<10 30 50>
    windFrequency=<10 3 4>
    windOffset=<0 -200 -250>
   }
   onGround_Crouch_Moving {
    maxDownInterpSpeed=5
    maxGustDuration=2
    maxGustStrength=7
    maxTimeBetweenGusts=7
    maxUpInterpSpeed=5
    minDownInterpSpeed=2
    minGustDuration=0.5
    minGustStrength=3
    minTimeBetweenGusts=0
    minUpInterpSpeed=2
    windTurbulenceScalarA=0.1
    windTurbulenceScalarB=1
    windAmplitude=<100 25 75>
    windFrequency=<1.5 0.5 7.5>
    windOffset=<0 -100 -175>
   }
   onGround_Falling {
    maxDownInterpSpeed=5
    maxGustDuration=2
    maxGustStrength=7
    maxTimeBetweenGusts=7
    maxUpInterpSpeed=5
    minDownInterpSpeed=2
    minGustDuration=0.5
    minGustStrength=3
    minTimeBetweenGusts=0
    minUpInterpSpeed=2
    windTurbulenceScalarA=0.1
    windTurbulenceScalarB=1
    windAmplitude=<0 0 0>
    windFrequency=<0 0 0>
    windOffset=<0 0 0>
   }
   onGround_Moving {
    maxDownInterpSpeed=5
    maxGustDuration=8
    maxGustStrength=7
    maxTimeBetweenGusts=8
    maxUpInterpSpeed=5
    minDownInterpSpeed=2
    minGustDuration=0.2
    minGustStrength=3
    minTimeBetweenGusts=1
    minUpInterpSpeed=2
    windTurbulenceScalarA=0.1
    windTurbulenceScalarB=0.8
    windAmplitude=<30 65 42>
    windFrequency=<5 6 8.6>
    windOffset=<30 -250 -300>
   }
   onGround_Standing {
    maxDownInterpSpeed=5
    maxGustDuration=8
    maxGustStrength=7
    maxTimeBetweenGusts=8
    maxUpInterpSpeed=5
    minDownInterpSpeed=2
    minGustDuration=0.2
    minGustStrength=2
    minTimeBetweenGusts=1
    minUpInterpSpeed=2
    windTurbulenceScalarA=0.1
    windTurbulenceScalarB=0.8
    windAmplitude=<75 24 28>
    windFrequency=<1.5 2.3 3.7>
    windOffset=<75 -100 -150>
   }
   rail_Grinding {
    maxDownInterpSpeed=5
    maxGustDuration=8
    maxGustStrength=7
    maxTimeBetweenGusts=8
    maxUpInterpSpeed=5
    minDownInterpSpeed=2
    minGustDuration=0.2
    minGustStrength=3
    minTimeBetweenGusts=1
    minUpInterpSpeed=2
    windTurbulenceScalarA=0.1
    windTurbulenceScalarB=0.8
    windAmplitude=<20 65 42>
    windFrequency=<5 6 8.6>
    windOffset=<50 -500 -400>
   }
   riding_Moving {
    maxDownInterpSpeed=5
    maxGustDuration=2
    maxGustStrength=7
    maxTimeBetweenGusts=7
    maxUpInterpSpeed=5
    minDownInterpSpeed=2
    minGustDuration=0.5
    minGustStrength=3
    minTimeBetweenGusts=0
    minUpInterpSpeed=2
    windTurbulenceScalarA=0.1
    windTurbulenceScalarB=1
    windAmplitude=<0 0 0>
    windFrequency=<0 0 0>
    windOffset=<0 0 0>
   }
   riding_Moving_Targeting {
    maxDownInterpSpeed=5
    maxGustDuration=2
    maxGustStrength=7
    maxTimeBetweenGusts=7
    maxUpInterpSpeed=5
    minDownInterpSpeed=2
    minGustDuration=0.5
    minGustStrength=3
    minTimeBetweenGusts=0
    minUpInterpSpeed=2
    windTurbulenceScalarA=0.1
    windTurbulenceScalarB=1
    windAmplitude=<0 0 0>
    windFrequency=<0 0 0>
    windOffset=<0 0 0>
   }
   riding_Standing {
    maxDownInterpSpeed=5
    maxGustDuration=2
    maxGustStrength=7
    maxTimeBetweenGusts=7
    maxUpInterpSpeed=5
    minDownInterpSpeed=2
    minGustDuration=0.5
    minGustStrength=3
    minTimeBetweenGusts=0
    minUpInterpSpeed=2
    windTurbulenceScalarA=0.1
    windTurbulenceScalarB=1
    windAmplitude=<0 0 0>
    windFrequency=<0 0 0>
    windOffset=<0 0 0>
   }
   skydive_DivingDown {
    maxDownInterpSpeed=5
    maxGustDuration=8
    maxGustStrength=7
    maxTimeBetweenGusts=8
    maxUpInterpSpeed=5
    minDownInterpSpeed=2
    minGustDuration=0.2
    minGustStrength=3
    minTimeBetweenGusts=1
    minUpInterpSpeed=2
    windTurbulenceScalarA=0.1
    windTurbulenceScalarB=4
    windAmplitude=<0 25 35>
    windFrequency=<0 4 10>
    windOffset=<-450 -100 100>
   }
   skydive_FloatingDown {
    maxDownInterpSpeed=5
    maxGustDuration=8
    maxGustStrength=1
    maxTimeBetweenGusts=8
    maxUpInterpSpeed=5
    minDownInterpSpeed=2
    minGustDuration=0.2
    minGustStrength=0
    minTimeBetweenGusts=1
    minUpInterpSpeed=2
    windTurbulenceScalarA=0.5
    windTurbulenceScalarB=1
    windAmplitude=<50 20 35>
    windFrequency=<8 2 3>
    windOffset=<0 -300 200>
   }
   skydive_GoingUp {
    maxDownInterpSpeed=5
    maxGustDuration=8
    maxGustStrength=7
    maxTimeBetweenGusts=8
    maxUpInterpSpeed=5
    minDownInterpSpeed=2
    minGustDuration=0.2
    minGustStrength=3
    minTimeBetweenGusts=1
    minUpInterpSpeed=2
    windTurbulenceScalarA=0.1
    windTurbulenceScalarB=0.8
    windAmplitude=<0 0 0>
    windFrequency=<0 0 0>
    windOffset=<0 -500 300>
   }
   skydive_Parachuting {
    maxDownInterpSpeed=5
    maxGustDuration=8
    maxGustStrength=1
    maxTimeBetweenGusts=8
    maxUpInterpSpeed=5
    minDownInterpSpeed=2
    minGustDuration=5
    minGustStrength=0
    minTimeBetweenGusts=5
    minUpInterpSpeed=2
    windTurbulenceScalarA=0.2
    windTurbulenceScalarB=1
    windAmplitude=<25 50 75>
    windFrequency=<8 6 2>
    windOffset=<-150 -400 -150>
   }
   skydive_Surfing {
    maxDownInterpSpeed=5
    maxGustDuration=2
    maxGustStrength=7
    maxTimeBetweenGusts=7
    maxUpInterpSpeed=5
    minDownInterpSpeed=2
    minGustDuration=0.5
    minGustStrength=3
    minTimeBetweenGusts=0
    minUpInterpSpeed=2
    windTurbulenceScalarA=0.1
    windTurbulenceScalarB=1
    windAmplitude=<0 0 0>
    windFrequency=<0 0 0>
    windOffset=<0 0 0>
   }
   swimming {
    maxDownInterpSpeed=5
    maxGustDuration=2
    maxGustStrength=7
    maxTimeBetweenGusts=7
    maxUpInterpSpeed=5
    minDownInterpSpeed=2
    minGustDuration=0.5
    minGustStrength=3
    minTimeBetweenGusts=0
    minUpInterpSpeed=2
    windTurbulenceScalarA=0.1
    windTurbulenceScalarB=1
    windAmplitude=<10 25 20>
    windFrequency=<1.5 0.5 7.5>
    windOffset=<0 -150 50>
   }
   tactical_Sprint {
    maxDownInterpSpeed=5
    maxGustDuration=2
    maxGustStrength=7
    maxTimeBetweenGusts=7
    maxUpInterpSpeed=5
    minDownInterpSpeed=2
    minGustDuration=0.5
    minGustStrength=3
    minTimeBetweenGusts=0
    minUpInterpSpeed=2
    windTurbulenceScalarA=0.1
    windTurbulenceScalarB=1
    windAmplitude=<0 0 0>
    windFrequency=<0 0 0>
    windOffset=<0 0 0>
   }
   ziplining {
    maxDownInterpSpeed=5
    maxGustDuration=2
    maxGustStrength=7
    maxTimeBetweenGusts=7
    maxUpInterpSpeed=5
    minDownInterpSpeed=2
    minGustDuration=0.5
    minGustStrength=3
    minTimeBetweenGusts=0
    minUpInterpSpeed=2
    windTurbulenceScalarA=0.1
    windTurbulenceScalarB=1
    windAmplitude=<0 0 0>
    windFrequency=<0 0 0>
    windOffset=<0 0 0>
   }
  }
 }
]
