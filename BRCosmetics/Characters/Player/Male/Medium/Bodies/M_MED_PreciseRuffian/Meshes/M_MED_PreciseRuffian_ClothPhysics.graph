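# Physics asset for the PreciseRuffian medium male player body, named for cloth physics and apparently serving as collision for its simulated clothing.
# Capsules on the pelvis and two lower spine bones cover the torso, and a large box on the root sits low like a ground slab.
# Forearm, thigh, calf and foot bodies list no shapes.
# Constraints chain the bodies along the skeleton with all swing and twist motion free.
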PhysicsAsset /BRCosmetics/Characters/Player/Male/Medium/Bodies/M_MED_PreciseRuffian/Meshes/M_MED_PreciseRuffian_ClothPhysics
bodies {
 calf_l []
 calf_r []
 foot_l []
 foot_r []
 lowerarm_l []
 lowerarm_r []
 pelvis [{shapeType=Capsule, length=7.9865, radius=15, center=<-3.43 0.3517 0>, rotation=<rot 0 -0.0008 -0.0007>}]
 root [{shapeType=Box, extentX=327.5, extentY=327.5, extentZ=59.22, shapeName=root_box, center=<0 0 -26.4087>}]
 spine_01 [{shapeType=Capsule, length=7.9663, radius=10.1267, center=<0.3095 -0.6696 0>}]
 spine_02 [{shapeType=Capsule, length=8.2792, radius=9.2496, center=<-1.0816 -0.436 -0.0002>}]
 thigh_l []
 thigh_r []
}
constraints [
 {
  bone1Name=foot_r
  bone2Name=calf_r
  swing1LimitDegrees=45
  swing1Motion=Free
  swing2LimitDegrees=45
  swing2Motion=Free
  twistLimitDegrees=45
  twistMotion=Free
 }
 {
  bone1Name=pelvis
  bone2Name=root
  swing1LimitDegrees=45
  swing1Motion=Free
  swing2LimitDegrees=45
  swing2Motion=Free
  twistLimitDegrees=45
  twistMotion=Free
 }
 {
  bone1Name=spine_01
  bone2Name=pelvis
  swing1LimitDegrees=45
  swing1Motion=Free
  swing2LimitDegrees=45
  swing2Motion=Free
  twistLimitDegrees=45
  twistMotion=Free
 }
 {
  bone1Name=spine_02
  bone2Name=spine_01
  swing1LimitDegrees=45
  swing1Motion=Free
  swing2LimitDegrees=45
  swing2Motion=Free
  twistLimitDegrees=45
  twistMotion=Free
 }
 {
  bone1Name=lowerarm_l
  bone2Name=spine_02
  swing1LimitDegrees=45
  swing1Motion=Free
  swing2LimitDegrees=45
  swing2Motion=Free
  twistLimitDegrees=45
  twistMotion=Free
 }
 {
  bone1Name=lowerarm_r
  bone2Name=spine_02
  swing1LimitDegrees=45
  swing1Motion=Free
  swing2LimitDegrees=45
  swing2Motion=Free
  twistLimitDegrees=45
  twistMotion=Free
 }
 {
  bone1Name=thigh_l
  bone2Name=pelvis
  swing1LimitDegrees=45
  swing1Motion=Free
  swing2LimitDegrees=45
  swing2Motion=Free
  twistLimitDegrees=45
  twistMotion=Free
 }
 {
  bone1Name=calf_l
  bone2Name=thigh_l
  swing1LimitDegrees=45
  swing1Motion=Free
  swing2LimitDegrees=45
  swing2Motion=Free
  twistLimitDegrees=45
  twistMotion=Free
 }
 {
  bone1Name=foot_l
  bone2Name=calf_l
  swing1LimitDegrees=45
  swing1Motion=Free
  swing2LimitDegrees=45
  swing2Motion=Free
  twistLimitDegrees=45
  twistMotion=Free
 }
 {
  bone1Name=thigh_r
  bone2Name=pelvis
  swing1LimitDegrees=45
  swing1Motion=Free
  swing2LimitDegrees=45
  swing2Motion=Free
  twistLimitDegrees=45
  twistMotion=Free
 }
 {
  bone1Name=calf_r
  bone2Name=thigh_r
  swing1LimitDegrees=45
  swing1Motion=Free
  swing2LimitDegrees=45
  swing2Motion=Free
  twistLimitDegrees=45
  twistMotion=Free
 }
]
