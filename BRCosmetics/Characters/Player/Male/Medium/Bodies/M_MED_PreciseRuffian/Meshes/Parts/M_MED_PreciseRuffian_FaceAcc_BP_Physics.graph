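# Physics asset for the PreciseRuffian male medium face accessory, defining secondary motion for its ears and two headband tails.
# Capsules on upper spine, shoulders, upper arms, neck and head form a locked body chain that anchors the accessory bases.
# Each side has a two-bone ear chain with tight angular limits and a five-bone headband chain with limited swing and locked twist.
# Duplicate neck and head helper-joint capsules hang from spine_05 on free angular constraints.
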PhysicsAsset /BRCosmetics/Characters/Player/Male/Medium/Bodies/M_MED_PreciseRuffian/Meshes/Parts/M_MED_PreciseRuffian_FaceAcc_BP_Physics
bodies {
 base_ear_00_l [{shapeType=Capsule, length=0, radius=0, shapeName=base_ear_00_l_capsule, rotation=<rot 90 0 0>}]
 base_ear_00_r [{shapeType=Capsule, length=0, radius=0, shapeName=base_ear_00_r_capsule, rotation=<rot 90 0 0>}]
 base_headBand_00_l [{shapeType=Capsule, length=1, radius=1, shapeName=base_headBand_00_l_capsule, rotation=<rot 90 0 0>}]
 base_headBand_00_r [{shapeType=Capsule, length=1, radius=1, shapeName=base_headBand_00_r_capsule, rotation=<rot 90 0 0>}]
 base_headBand_c [{shapeType=Capsule, length=1, radius=1, shapeName=base_headBand_c_capsule, rotation=<rot 90 0 0>}]
 clavicle_l [{
   shapeType=Capsule
   length=8
   radius=17.5
   shapeName=clavicle_l_capsule
   center=<6.8059 5.7877 -1.3366>
   rotation=<rot 68.8612 -0.0002 -6.6463>
  }]
 clavicle_r [{
   shapeType=Capsule
   length=8
   radius=17.5
   shapeName=clavicle_l_capsule
   center=<-6.8059 -5.7876 1.3366>
   rotation=<rot 68.8612 -0.0002 -6.6463>
  }]
 dyn_ear_00_l [{shapeType=Capsule, length=2, radius=2, shapeName=dyn_ear_00_l_capsule, center=<3 0 0>, rotation=<rot 90 0 0>}]
 dyn_ear_00_r [{shapeType=Capsule, length=2, radius=2, shapeName=dyn_ear_00_r_capsule, center=<3 0 0>, rotation=<rot 90 0 0>}]
 dyn_ear_01_l [{shapeType=Capsule, length=2, radius=1.5, shapeName=dyn_ear_01_l_capsule, center=<3 0 0>, rotation=<rot 90 0 0>}]
 dyn_ear_01_r [{shapeType=Capsule, length=2, radius=1.5, shapeName=dyn_ear_01_r_capsule, center=<3 0 0>, rotation=<rot 90 0 0>}]
 dyn_headBand_00_l [{shapeType=Capsule, length=2.5, radius=1.5, shapeName=dyn_headBand_00_l_capsule, center=<2 -0.45 0>, rotation=<rot 90 0 0>}]
 dyn_headBand_00_r [{shapeType=Capsule, length=2.5, radius=1.5, shapeName=dyn_headBand_00_r_capsule, center=<2 -0.45 0>, rotation=<rot 90 0 0>}]
 dyn_headBand_01_l [{shapeType=Capsule, length=4.5, radius=1.25, shapeName=dyn_headBand_01_l_capsule, center=<3 0 -0.3645>, rotation=<rot 90 0 0>}]
 dyn_headBand_01_r [{shapeType=Capsule, length=4.5, radius=1.25, shapeName=dyn_headBand_01_r_capsule, center=<3 0 0>, rotation=<rot 90 0 0>}]
 dyn_headBand_02_l [{shapeType=Capsule, length=4.5, radius=1.6, shapeName=dyn_headBand_02_l_capsule, center=<4 -0.4 0>, rotation=<rot 90 0 0>}]
 dyn_headBand_02_r [{shapeType=Capsule, length=6, radius=1.25, shapeName=dyn_headBand_02_r_capsule, center=<4 -0.4 0>, rotation=<rot 90 0 0>}]
 dyn_headBand_03_l [{shapeType=Capsule, length=7.5, radius=1.5, shapeName=dyn_headBand_03_l_capsule, center=<4.9325 0 0>, rotation=<rot 90 0 0>}]
 dyn_headBand_03_r [{shapeType=Capsule, length=6, radius=2, shapeName=dyn_headBand_03_r_capsule, center=<4 0 0>, rotation=<rot 90 0 0>}]
 dyn_headBand_04_l [{shapeType=Capsule, length=6.5, radius=1.5, shapeName=dyn_headBand_04_l_capsule, center=<4.2209 0 0>, rotation=<rot 90 0 0>}]
 dyn_headBand_04_r [{shapeType=Capsule, length=8.5, radius=2, shapeName=dyn_headBand_04_r_capsule, center=<5.6047 -0.4 0>, rotation=<rot 90 0 0>}]
 head [{shapeType=Capsule, length=4, radius=7.5, shapeName=head_capsule, center=<7.5431 2.4073 0>, rotation=<rot 90 0 0>}]
 jnt_help_head [{shapeType=Capsule, length=4, radius=7.5, shapeName=head_capsule, center=<7.5431 2.8165 0>, rotation=<rot 90 0 0>}]
 jnt_help_neck_01 [{shapeType=Capsule, length=9, radius=12.5, shapeName=neck_01_capsule, center=<2.9301 2.7749 0>}]
 jnt_help_neck_02 [{shapeType=Capsule, length=4, radius=10, shapeName=neck_02_capsule, center=<2.9089 0.4092 0>}]
 neck_01 [{shapeType=Capsule, length=67, radius=12.5, shapeName=neck_01_capsule, center=<2.9301 2.3657 0>}]
 neck_02 [{shapeType=Capsule, length=62, radius=10, shapeName=neck_02_capsule, center=<2.9089 -0.4777 0>}]
 spine_03 [{shapeType=Capsule, length=5, radius=17, shapeName=spine_03_capsule, center=<0 3.5322 0>}]
 spine_04 [{shapeType=Capsule, length=4.5, radius=17.5, shapeName=spine_04_capsule, center=<0 1.7377 0>}]
 spine_05 [{shapeType=Capsule, length=10.5, radius=18.5, shapeName=spine_05_capsule, center=<0 4.4538 0>}]
 upperarm_l [{shapeType=Capsule, length=15, radius=12, shapeName=upperarm_l_capsule, center=<5.4811 1.9436 -1.643>, rotation=<rot 90 180 174.6447>}]
 upperarm_r [{shapeType=Capsule, length=15, radius=12, shapeName=upperarm_l_capsule, center=<-5.4811 -1.9436 1.643>, rotation=<rot 90 180 174.6447>}]
}
constraints [
 {
  bone1Name=spine_04
  bone2Name=spine_03
  swing1LimitDegrees=45
  swing1Motion=Locked
  swing2LimitDegrees=45
  swing2Motion=Locked
  twistLimitDegrees=45
  twistMotion=Locked
 }
 {
  bone1Name=spine_05
  bone2Name=spine_04
  swing1LimitDegrees=45
  swing1Motion=Locked
  swing2LimitDegrees=45
  swing2Motion=Locked
  twistLimitDegrees=45
  twistMotion=Locked
 }
 {
  bone1Name=clavicle_l
  bone2Name=spine_05
  swing1LimitDegrees=45
  swing1Motion=Locked
  swing2LimitDegrees=45
  swing2Motion=Locked
  twistLimitDegrees=45
  twistMotion=Locked
 }
 {
  bone1Name=upperarm_l
  bone2Name=clavicle_l
  swing1LimitDegrees=45
  swing1Motion=Locked
  swing2LimitDegrees=45
  swing2Motion=Locked
  twistLimitDegrees=45
  twistMotion=Locked
 }
 {
  bone1Name=clavicle_r
  bone2Name=spine_05
  swing1LimitDegrees=45
  swing1Motion=Locked
  swing2LimitDegrees=45
  swing2Motion=Locked
  twistLimitDegrees=45
  twistMotion=Locked
 }
 {
  bone1Name=upperarm_r
  bone2Name=clavicle_r
  swing1LimitDegrees=45
  swing1Motion=Locked
  swing2LimitDegrees=45
  swing2Motion=Locked
  twistLimitDegrees=45
  twistMotion=Locked
 }
 {
  bone1Name=neck_01
  bone2Name=spine_05
  swing1LimitDegrees=45
  swing1Motion=Locked
  swing2LimitDegrees=45
  swing2Motion=Locked
  twistLimitDegrees=45
  twistMotion=Locked
 }
 {
  bone1Name=neck_02
  bone2Name=neck_01
  swing1LimitDegrees=45
  swing1Motion=Locked
  swing2LimitDegrees=45
  swing2Motion=Locked
  twistLimitDegrees=45
  twistMotion=Locked
 }
 {
  bone1Name=head
  bone2Name=neck_02
  swing1LimitDegrees=45
  swing1Motion=Locked
  swing2LimitDegrees=45
  swing2Motion=Locked
  twistLimitDegrees=45
  twistMotion=Locked
 }
 {
  bone1Name=base_ear_00_l
  bone2Name=head
  swing1LimitDegrees=45
  swing1Motion=Locked
  swing2LimitDegrees=45
  swing2Motion=Locked
  twistLimitDegrees=45
  twistMotion=Locked
 }
 {bone1Name=dyn_ear_00_l, bone2Name=base_ear_00_l, swing1LimitDegrees=5, swing2LimitDegrees=5, twistLimitDegrees=5}
 {bone1Name=dyn_ear_01_l, bone2Name=dyn_ear_00_l, swing1LimitDegrees=7, swing2LimitDegrees=7, twistLimitDegrees=2}
 {
  bone1Name=base_ear_00_r
  bone2Name=head
  swing1LimitDegrees=45
  swing1Motion=Locked
  swing2LimitDegrees=45
  swing2Motion=Locked
  twistLimitDegrees=45
  twistMotion=Locked
 }
 {bone1Name=dyn_ear_00_r, bone2Name=base_ear_00_r, swing1LimitDegrees=5, swing2LimitDegrees=5, twistLimitDegrees=5}
 {bone1Name=dyn_ear_01_r, bone2Name=dyn_ear_00_r, swing1LimitDegrees=7, swing2LimitDegrees=7, twistLimitDegrees=2}
 {
  bone1Name=base_headBand_c
  bone2Name=head
  swing1LimitDegrees=45
  swing1Motion=Locked
  swing2LimitDegrees=45
  swing2Motion=Locked
  twistLimitDegrees=45
  twistMotion=Locked
 }
 {
  bone1Name=base_headBand_00_l
  bone2Name=base_headBand_c
  swing1LimitDegrees=45
  swing1Motion=Locked
  swing2LimitDegrees=45
  swing2Motion=Locked
  twistLimitDegrees=45
  twistMotion=Locked
 }
 {
  bone1Name=dyn_headBand_00_l
  bone2Name=base_headBand_00_l
  swing1LimitDegrees=25
  swing2LimitDegrees=10
  twistLimitDegrees=10
  twistMotion=Locked
 }
 {
  bone1Name=dyn_headBand_01_l
  bone2Name=dyn_headBand_00_l
  swing1LimitDegrees=35
  swing2LimitDegrees=12
  twistLimitDegrees=10
  twistMotion=Locked
 }
 {
  bone1Name=dyn_headBand_02_l
  bone2Name=dyn_headBand_01_l
  swing1LimitDegrees=13
  swing2LimitDegrees=20
  twistLimitDegrees=10
  twistMotion=Locked
 }
 {
  bone1Name=dyn_headBand_03_l
  bone2Name=dyn_headBand_02_l
  swing1LimitDegrees=13
  swing2LimitDegrees=20
  twistLimitDegrees=5
  twistMotion=Locked
 }
 {
  bone1Name=dyn_headBand_04_l
  bone2Name=dyn_headBand_03_l
  swing1LimitDegrees=13
  swing2LimitDegrees=20
  twistLimitDegrees=5
  twistMotion=Locked
 }
 {
  bone1Name=base_headBand_00_r
  bone2Name=base_headBand_c
  swing1LimitDegrees=45
  swing1Motion=Locked
  swing2LimitDegrees=45
  swing2Motion=Locked
  twistLimitDegrees=45
  twistMotion=Locked
 }
 {
  bone1Name=dyn_headBand_00_r
  bone2Name=base_headBand_00_r
  swing1LimitDegrees=25
  swing2LimitDegrees=10
  twistLimitDegrees=10
  twistMotion=Locked
 }
 {
  bone1Name=dyn_headBand_01_r
  bone2Name=dyn_headBand_00_r
  swing1LimitDegrees=35
  swing2LimitDegrees=12
  twistLimitDegrees=10
  twistMotion=Locked
 }
 {
  bone1Name=dyn_headBand_02_r
  bone2Name=dyn_headBand_01_r
  swing1LimitDegrees=13
  swing2LimitDegrees=20
  twistLimitDegrees=10
  twistMotion=Locked
 }
 {
  bone1Name=dyn_headBand_03_r
  bone2Name=dyn_headBand_02_r
  swing1LimitDegrees=13
  swing2LimitDegrees=20
  twistLimitDegrees=5
  twistMotion=Locked
 }
 {
  bone1Name=dyn_headBand_04_r
  bone2Name=dyn_headBand_03_r
  swing1LimitDegrees=13
  swing2LimitDegrees=20
  twistLimitDegrees=5
  twistMotion=Locked
 }
 {
  bone1Name=jnt_help_neck_01
  bone2Name=spine_05
  swing1LimitDegrees=45
  swing1Motion=Free
  swing2LimitDegrees=45
  swing2Motion=Free
  twistLimitDegrees=45
  twistMotion=Free
 }
 {
  bone1Name=jnt_help_neck_02
  bone2Name=jnt_help_neck_01
  swing1LimitDegrees=45
  swing1Motion=Free
  swing2LimitDegrees=45
  swing2Motion=Free
  twistLimitDegrees=45
  twistMotion=Free
 }
 {
  bone1Name=jnt_help_head
  bone2Name=jnt_help_neck_02
  swing1LimitDegrees=45
  swing1Motion=Free
  swing2LimitDegrees=45
  swing2Motion=Free
  twistLimitDegrees=45
  twistMotion=Free
 }
]
