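# Post-process-style animation for the Precise Ruffian face accessory, posing its headband and tails over pose copied from the parent and body meshes.
# With rigid bodies enabled, head-pitch offsets feed a simulation with state-driven gravity; wearing backbling switches to constrained, rotated tails with half-strength physics.
# With rigid bodies disabled, trail controllers weighted by player movement replace simulation, with head-angle-driven rotation at the tail roots.
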AnimBlueprint /BRCosmetics/Characters/Player/Male/Medium/Bodies/M_MED_PreciseRuffian/Meshes/Parts/M_MED_PreciseRuffian_FaceAcc_AnimBP
target_skeleton=&/M_MED_PreciseRuffian_FaceAcc_Skeleton
graphs [
 AnimGraph {nodes [
   n1 = "Transform (Modify) Bone - Bone: base_headBand_c"!(
    ComponentPose=Local To Component(LocalPose=Use cached pose 'StartCache'())
    Alpha=Get Head Up() -> (Head Up:double)
    Translation="-3,0,0"
    Rotation="0,-60,0"
   ) -> (Pose) {
    rotationMode=BMM_Additive
    rotationSpace=BCS_BoneSpace
    translationMode=BMM_Additive
    translationSpace=BCS_BoneSpace
    alphaBoolBlend {blendOption=Linear}
    boneToModify {boneName=base_headBand_c}
    rotation {yaw=-60}
    translation {x=-3}
   }
   n2 = "Transform (Modify) Bone - Bone: base_headBand_c"(ComponentPose=@n1, Alpha=Get Head Down() -> (Head Down:double), Translation="10,0,0", Rotation="0,60,0") {
    rotationMode=BMM_Additive
    rotationSpace=BCS_BoneSpace
    translationMode=BMM_Additive
    translationSpace=BCS_BoneSpace
    alphaBoolBlend {blendOption=Linear}
    boneToModify {boneName=base_headBand_c}
    rotation {yaw=60}
    translation {x=10}
   }
   n3 = RigidBody!(ComponentPose=@n2, Alpha, OverrideWorldGravity:Vector=Get GravityOverride(), ExternalForce) -> (Pose) {
    bClampLinearTranslationLimitToRefPose=true
    bOverrideWorldGravity=true
    overridePhysicsAsset=&/M_MED_PreciseRuffian_FaceAcc_Physics
    simulationSpace=BaseBoneSpace
    alphaBoolBlend {blendOption=Linear}
    baseBoneRef {boneName=spine_05}
    componentAppliedLinearAccClamp=<3000 3000 3000>
    componentLinearAccScale=<0.3 0.3 0.5>
    componentLinearVelScale=<0.3 0.3 0.5>
    simSpaceSettings {maxAngularAcceleration=1000, maxAngularVelocity=3, maxLinearAcceleration=700, maxLinearVelocity=500, velocityScaleZ=0.2, worldAlpha=0.25}
   }
   n4 = Component To Local(ComponentPose=@n3)
   n5 = "Constraint - dyn_headBand_00_l (P)"(ComponentPose=Local To Component(LocalPose=Use cached pose 'StartCache'()), Alpha=1, ConstraintWeights_0=1) {
    alphaBoolBlend {blendOption=Linear}
    boneToModify {boneName=dyn_headBand_00_l}
    constraintSetup [{targetBone {boneName=neck_01}}]
    constraintWeights [1]
   }
   n6 = "Constraint - dyn_headBand_01_l (P)"(ComponentPose=@n5, Alpha=1, ConstraintWeights_0=1) {
    alphaBoolBlend {blendOption=Linear}
    boneToModify {boneName=dyn_headBand_01_l}
    constraintSetup [{targetBone {boneName=neck_02}}]
    constraintWeights [1]
   }
   n7 = "Constraint - dyn_headBand_02_l (P)"(ComponentPose=@n6, Alpha=1, ConstraintWeights_0=1) {
    alphaBoolBlend {blendOption=Linear}
    boneToModify {boneName=dyn_headBand_02_l}
    constraintSetup [{targetBone {boneName=neck_02}}]
    constraintWeights [1]
   }
   n8 = "Constraint - dyn_headBand_00_r (P)"(ComponentPose=@n7, Alpha=1, ConstraintWeights_0=1) {
    alphaBoolBlend {blendOption=Linear}
    boneToModify {boneName=dyn_headBand_00_r}
    constraintSetup [{targetBone {boneName=dyn_headBand_00_l}}]
    constraintWeights [1]
   }
   n9 = "Constraint - dyn_headBand_01_r (P)"(ComponentPose=@n8, Alpha=1, ConstraintWeights_0=1) {
    alphaBoolBlend {blendOption=Linear}
    boneToModify {boneName=dyn_headBand_01_r}
    constraintSetup [{targetBone {boneName=dyn_headBand_01_l}}]
    constraintWeights [1]
   }
   n10 = "Constraint - dyn_headBand_02_l (P)"(ComponentPose=@n9, Alpha=1, ConstraintWeights_0=1) {
    alphaBoolBlend {blendOption=Linear}
    boneToModify {boneName=dyn_headBand_02_l}
    constraintSetup [{targetBone {boneName=neck_02}}]
    constraintWeights [1]
   }
   n11 = "Transform (Modify) Bone - Bone: dyn_headBand_01_l"!(ComponentPose=@n10, Alpha=1, Translation="0,0,0", Rotation="0,45,0") -> (Pose) {
    rotationMode=BMM_Additive
    rotationSpace=BCS_BoneSpace
    translationMode=BMM_Additive
    translationSpace=BCS_BoneSpace
    alphaBoolBlend {blendOption=Linear}
    boneToModify {boneName=dyn_headBand_01_l}
    rotation {yaw=45}
   }
   n12 = "Transform (Modify) Bone - Bone: dyn_headBand_01_r"!(ComponentPose=@n11, Alpha=1, Translation="0,0,0", Rotation="15,45,0") -> (Pose) {
    rotationMode=BMM_Additive
    rotationSpace=BCS_BoneSpace
    translationMode=BMM_Additive
    translationSpace=BCS_BoneSpace
    alphaBoolBlend {blendOption=Linear}
    boneToModify {boneName=dyn_headBand_01_r}
    rotation {pitch=15, yaw=45}
   }
   n13 = "Transform (Modify) Bone - Bone: dyn_headBand_03_l"!(ComponentPose=@n12, Alpha=1, Translation="0,0,0", Rotation="0,35,0") -> (Pose) {
    rotationMode=BMM_Additive
    rotationSpace=BCS_BoneSpace
    translationMode=BMM_Additive
    translationSpace=BCS_BoneSpace
    alphaBoolBlend {blendOption=Linear}
    boneToModify {boneName=dyn_headBand_03_l}
    rotation {yaw=35}
   }
   n14 = "Transform (Modify) Bone - Bone: dyn_headBand_03_r"!(ComponentPose=@n13, Alpha=1, Translation="0,0,0", Rotation="0,25,0") -> (Pose) {
    rotationMode=BMM_Additive
    rotationSpace=BCS_BoneSpace
    translationMode=BMM_Additive
    translationSpace=BCS_BoneSpace
    alphaBoolBlend {blendOption=Linear}
    boneToModify {boneName=dyn_headBand_03_r}
    rotation {yaw=25}
   }
   n15 = RigidBody(ComponentPose=@n14, Alpha=0.5, OverrideWorldGravity="0,0,-1280") {
    bClampLinearTranslationLimitToRefPose=true
    bOverrideWorldGravity=true
    overridePhysicsAsset=&/M_MED_PreciseRuffian_FaceAcc_BP_Physics
    simulationSpace=BaseBoneSpace
    alphaBoolBlend {blendOption=Linear}
    baseBoneRef {boneName=spine_05}
    componentAppliedLinearAccClamp=<3000 3000 3000>
    simSpaceSettings {maxAngularAcceleration=1000, maxAngularVelocity=3, maxLinearAcceleration=700, maxLinearVelocity=500, velocityScaleZ=0}
   }
   n16 = Component To Local(ComponentPose=@n15)
   n17 = Get isWearingBackbling()
   n18 = "Transform (Modify) Bone - Bone: base_headBand_c"!(ComponentPose=Local To Component(LocalPose=Use cached pose 'StartCache'()), Alpha=1, Rotation="0,0,0") -> (Pose) {
    rotationMode=BMM_Additive
    rotationSpace=BCS_BoneSpace
    translationMode=BMM_Additive
    translationSpace=BCS_BoneSpace
    alphaBoolBlend {blendOption=Linear}
    boneToModify {boneName=base_headBand_c}
    translation=<-3.0307 -1.9821 0.3735>
   }
   n19 = "Transform (Modify) Bone - Bone: dyn_headBand_01_l"!(ComponentPose=@n18, Alpha=1, Rotation="0,0,0") -> (Pose) {
    rotationMode=BMM_Additive
    rotationSpace=BCS_BoneSpace
    translationMode=BMM_Additive
    translationSpace=BCS_BoneSpace
    alphaBoolBlend {blendOption=Linear}
    boneToModify {boneName=dyn_headBand_01_l}
    translation {x=-5}
   }
   n20 = "Transform (Modify) Bone - Bone: dyn_headBand_01_r"!(ComponentPose=@n19, Alpha=1, Rotation="10,0,0") -> (Pose) {
    rotationMode=BMM_Additive
    rotationSpace=BCS_BoneSpace
    translationMode=BMM_Additive
    translationSpace=BCS_BoneSpace
    alphaBoolBlend {blendOption=Linear}
    boneToModify {boneName=dyn_headBand_01_r}
    rotation {pitch=10}
    translation {x=-8}
   }
   n21 = Get TrailControllerRot() -> (TrailControllerRot:Rotator)
   n22 = "Transform (Modify) Bone - Bone: dyn_headBand_00_l"!(ComponentPose=@n20, Alpha=1, Rotation=@n21) -> (Pose) {rotationMode=BMM_Additive, rotationSpace=BCS_BoneSpace, alphaBoolBlend {blendOption=Linear}, boneToModify {boneName=dyn_headBand_00_l}}
   n23 = "Transform (Modify) Bone - Bone: dyn_headBand_00_r"!(ComponentPose=@n22, Alpha=1, Rotation=@n21) -> (Pose) {rotationMode=BMM_Additive, rotationSpace=BCS_BoneSpace, alphaBoolBlend {blendOption=Linear}, boneToModify {boneName=dyn_headBand_00_r}}
   n24 = Get BwdMovement() -> (BwdMovement:double)
   n25 = Trail controller - Bone: dyn_headBand_04_l(ComponentPose=@n23, Alpha=@n24) {
    bUsePlanarLimit=true
    chainLength=5
    debugLifeTime=0.1
    maxDeltaTime=0.333
    alphaScaleBiasClamp {bClampResult=true, bInterpResult=true, clampMax=0.5}
    rotationLimits [{limitMax=<180 180 180>}, {limitMax=<180 180 180>}, {limitMax=<180 180 180>}, {limitMax=<180 180 180>}, {limitMax=<180 180 180>}]
    rotationOffsets [<0 0 0>, <0 0 0>, <0 0 0>, <0 0 0>, <0 0 0>]
    trailBone {boneName=dyn_headBand_04_l}
    trailRelaxationSpeed {editorCurveData {keys [<at -0.0134 19.3841 RCIM_Linear>,<at 0.5079 12.8986 RCIM_Linear>,<at 0.9955 9.6044 RCIM_Linear>]}}
   }
   n26 = Trail controller - Bone: dyn_headBand_04_r(ComponentPose=@n25, Alpha=@n24) {
    bUsePlanarLimit=true
    chainLength=5
    debugLifeTime=0.1
    maxDeltaTime=0.333
    alphaScaleBiasClamp {bClampResult=true, bInterpResult=true, clampMax=0.5}
    rotationLimits [{limitMax=<180 180 180>}, {limitMax=<180 180 180>}, {limitMax=<180 180 180>}, {limitMax=<180 180 180>}, {limitMax=<180 180 180>}]
    rotationOffsets [<0 0 0>, <0 0 0>, <0 0 0>, <0 0 0>, <0 0 0>]
    trailBone {boneName=dyn_headBand_04_r}
    trailRelaxationSpeed {editorCurveData {keys [<at -0.0134 19.3841 RCIM_Linear>,<at 0.5079 12.8986 RCIM_Linear>,<at 0.9955 9.6044 RCIM_Linear>]}}
   }
   Output Pose(Result=Blend Poses by bool(
     BlendPose_0=Blend Poses by bool(BlendPose_0=@n16, BlendPose_1=@n4, bActiveValue=@n17) {blendType=Linear}
     BlendPose_1=Component To Local(ComponentPose=@n26)
     bActiveValue=Get bRigidBodyEnabled()
    ) {blendType=Linear}) {layerGroup=None}
   n27 = Layered blend per bone(BasePose=Copy Pose From Mesh() {bUseAttachedParent=true}, BlendPoses_0=Copy Pose From Mesh(SourceMeshComponent=Get Body_Mesh())) {layerSetup [{branchFilters [
       {boneName=clavicle_l}
       {boneName=clavicle_r}
       {boneName=upperarm_l}
       {boneName=upperarm_r}
       {boneName=jnt_help_neck_01}
       {boneName=jnt_help_neck_02}
       {boneName=jnt_help_head}
      ]}]}
   Save cached pose 'StartCache'(Pose=@n27) {cacheName=StartCache}
  ]}
 EventGraph {nodes [
   Event BlueprintUpdateAnimation()
   > n1 = Sequence()
   > n2 = Gravity Override State Function DEPRECATED(CharacterPartAnimInstance=Self-Reference(), GameTimeInSeconds=GetGameTimeInSeconds(), GravityOverrideParams=Get Gravity Override Params())
   > Set GravityOverride(GravityOverride=@n2)
   > n3 = Branch(Condition=Get bRigidBodyEnabled())
   n4 = Get Skel Mesh()
   > Set Head Up(Head Up:double=GetAlphaFromReferenceTransformDelta(SkeletalMeshComponent=@n4, BoneName=spine_05, BaseName=head, InputRange=(X=0,Y=90), TwistAxis="0,0,1")) -> (Output_Get:double)
   > Set Head Down(Head Down:double=GetAlphaFromReferenceTransformDelta(SkeletalMeshComponent=@n4, BoneName=spine_05, BaseName=head, InputRange=(X=0,Y=-45), TwistAxis="0,0,1")) -> (Output_Get:double)
   n5 = GetAlphaFromReferenceTransformDelta(
    SkeletalMeshComponent=Get Skel Mesh()
    BoneName=spine_05
    BaseName=head
    InputRange=(X=-20,Y=20)
    OutputRange=(X=60,Y=-60)
    TwistAxis="0,0,1"
   )
   Set TrailControllerRot(execute=@n3.else, TrailControllerRot:Rotator=MakeRotator(Yaw="Clamp (Float)"(Value=@n5, Min=-10, Max=30))) -> (Output_Get:Rotator)
   > n6 = PlayerMovement(AnimInstance=Self-Reference(), XDotProductCoefficient=1, YDotProductCoefficient=1, ZDotProductCoefficient=1)
   n7 = MapRangeClamped(Value=@n6.YMovement, InRangeA=0, InRangeB=1)
   n8 = "Absolute (Float)"(A=@n6)
   > Set BwdMovement(BwdMovement:double=MapRangeClamped(Value=float + float(A=@n8, B=@n7) {is=title K2Node_CommutativeAssociativeBinaryOperator}, InRangeA=0, InRangeB=1.4)) -> (Output_Get:double)
   n9 = CheckForBackbling()
   Set isWearingBackbling(execute=@n1.then_1, isWearingBackbling=OR Boolean!(A=@n9, B=@n9.bIsBackblingACape, C:bool=@n9.bIsBackblingLong) -> (ReturnValue))
   Event OnCharacterCustomizationLoaded()
   > n10 = Cast To PlayerPawn_Athena(Object=TryGetPawnOwner())
   > Set Body_Mesh(Body_Mesh=GetSkeletalMeshForPartType(self=@n10))
   Event BlueprintInitializeAnimation()
   > Set Skel Mesh(Skel Mesh=GetOwnerSkeletalMeshComponent())
  ]}
]
variables [
 Body_Mesh {type=SkeletalMeshComponent}
 BwdMovement {type=number}
 Gravity Override Params {type=GravityOverrideParamsStruct, default {onGround_Standing {gravityOverride=<0 0 -700>}}}
 GravityOverride {type=Vector}
 Head Down {type=number}
 Head Up {type=number}
 isWearingBackbling {type=boolean}
 Skel Mesh {type=SkeletalMeshComponent}
 TrailControllerRot {type=Rotator}
]
